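# Control Rig for the DorsalDance large male player body, correcting deformation and placing accessories such as necklace, shorts, belt, watch, megaphone and gas can.
# A shared DeformRig function runs first, then a sequence constrains accessory and helper bones to weighted body bones.
# Additive bone offsets follow, weighted by DeformRig animation attributes and gameplay variables like parachute, skydive and crouch.
# Spherical pose readers on head, arms and right hand drive further fixes, some multiplied together or by weapon variables.
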ControlRigBlueprint /Game/Characters/Player/Male/Large/Bodies/M_LRG_DorsalDance/Meshes/M_LRG_DorsalDance_CtrlRig
graphs [
 Multiply_3_ContainedGraph {nodes [
   n1 = RigVMFunctionEntryNode()
   Multiply = &RigVMFunction_MathDoubleMul(A=@n1.A, B=@n1.B)
   RigVMFunctionReturnNode(Result=&RigVMFunction_MathDoubleMul(B=@n1.C, A=@Multiply.Result).Result)
  ]}
 New Function_ContainedGraph {nodes [
   RigVMFunctionEntryNode()
   > n1 = &RigUnit_SphericalPoseReader(
    ActiveRegionScaleFactors=(PositiveWidth=1,NegativeWidth=0.5,PositiveHeight=0.75,NegativeHeight=0.75)
    ActiveRegionSize=0.1
    Debug=(bDrawDebug=True,bDraw2D=False,bDrawLocalAxes=False,DebugScale=25,DebugSegments=20,DebugThickness=0.25)
    DriverAxis=(X=0,Y=1,Z=0)
    DriverItem=(Type=Bone,Name="head")
    FalloffRegionScaleFactors=(PositiveWidth=1,NegativeWidth=0.5,PositiveHeight=1,NegativeHeight=1)
    FalloffSize=0.2
    FlipHeightScaling=False
    FlipWidthScaling=False
    OptionalParentItem=(Type=Bone,Name="None")
    RotationOffset=(X=90,Y=0,Z=-30)
   )
   > RigVMVariableNode(Value=@n1.OutputParam) {variable=NewVar}
   > RigVMFunctionReturnNode()
  ]}
 New Function_ContainedGraph {nodes [
   RigVMFunctionEntryNode()
   > n1 = &RigUnit_SphericalPoseReader(
    ActiveRegionScaleFactors=(PositiveWidth=1,NegativeWidth=1,PositiveHeight=1,NegativeHeight=1)
    ActiveRegionSize=0.025
    Debug=(bDrawDebug=True,bDraw2D=False,bDrawLocalAxes=False,DebugScale=25,DebugSegments=20,DebugThickness=0.25)
    DriverAxis=(X=-1,Y=0,Z=0)
    DriverItem=(Type=Bone,Name="upperarm_r")
    FalloffRegionScaleFactors=(PositiveWidth=1,NegativeWidth=1,PositiveHeight=1,NegativeHeight=1)
    FalloffSize=0.1
    FlipHeightScaling=False
    FlipWidthScaling=False
    OptionalParentItem=(Type=Bone,Name="None")
    RotationOffset=(X=-115,Y=0,Z=25)
   )
   > RigVMVariableNode(Value=@n1.OutputParam) {variable=NewVar_0}
   > RigVMFunctionReturnNode()
  ]}
 RigVMFunctionLibrary {nodes [
   Get head_SPR_look_up_alpha = RigVMCollapseNode(ExecuteContext=())
   Get upperarm_r_SPR_aim_up_two_hands_alpha = RigVMCollapseNode(ExecuteContext=())
  ]}
 RigVMFunction_Sequence_1_ContainedGraph {nodes [
   n1 = RigVMFunctionEntryNode()
   RigVMFunction_Sequence = &RigVMFunction_Sequence(ExecuteContext=@n1.ExecuteContext)
   n2 = &RigVMFunction_Sequence(ExecuteContext=@RigVMFunction_Sequence.B)
   RigVMFunctionReturnNode(A=@RigVMFunction_Sequence.A, B=@n2.A, C=@n2.B)
  ]}
 RigVMModel {nodes [
   RigUnit_BeginExecution = &RigUnit_BeginExecution()
   FN_DeformRig_FNC = RigVMFunctionReferenceNode(
    DeformRig_ Curve="(deformRig_l_upperarm_up_alpha_72_#1=False,deformRig_r_upperarm_up_alpha_73_#2=true,deformRig_l_upperarm_fr_alpha_74_#3=False,deformRig_r_upperarm_fr_alpha_75_CB51B80 ...(truncated)"
    DeformRig_Bool="(upperarm_1_#4=True,elbow_3_#5=True,wrist_7_#6=True,groin_8_#7=True,glute_17_#8=True,knee_9_40C408324CA261 ...(truncated)"
    DeformRig_Debug=(upperarm_debug_18_#4=false,wrist_debug_20_#6=false,lat_debug_26_#9=false,groinGlute_debug_27_#7=false)
    DeformRig_Value="(upperarm_up_18_#10=(Rotation=(X=0,Y=0,Z=0,W=1),Translation=(X=0,Y=0,Z=8),Scale3D=(X=0.7,Y=1,Z=1)),upperarm_fwd_fr_19_#11=(Rotation=(X=0,Y=0,Z=0,W=1),Translation=(X=0,Y=10,Z=0),Scale3D=(X=0.5,Y=1,Z= ...(truncated)"
    ExecuteContext=@RigUnit_BeginExecution.ExecutePin
   )
   > RigVMFunction_Sequence = RigVMAggregateNode()
   DISPATCH_RigDispatch_GetAnimAttribute = RigVMDispatchNode(BoneName=root, Default=1, Name=deformRig_l_thigh_up_alpha) {template="DISPATCH_RigDispatch_GetAnimAttribute(in Name,in BoneName,in Default,out Value,out Found)"}
   &RigUnit_ModifyTransforms(
    ItemToModify="((Item=(Type=Bone,Name=\"base_belt_01_l\"),Transform=(Rotation=(X=0.1538,Y=-0.217,Z=0.4024,W=0.876),Translation=(X=2.5205,Y=13.0075,Z=-18.0748),Scale3D=(X=1,Y=1,Z=1))),(Item=(Type=Bone,Name=\"base_belt_01_r\"),Transform=(Rotation=(X=0,Y=0,Z=0,W=1),Translation= ...(truncated)"
    WeightMinimum=0
    Weight=@DISPATCH_RigDispatch_GetAnimAttribute.Value
    ExecutePin=@RigVMFunction_Sequence.B
   )
   > &RigUnit_ModifyTransforms(
    ItemToModify=((Item=(Type=Bone,Name="base_belt_01_r"),Transform=(Rotation=(X=0,Y=0,Z=0,W=1),Translation=(X=0,Y=6,Z=22),Scale3D=(X=1,Y=1,Z=1))))
    WeightMinimum=0
    Weight=RigVMDispatchNode(BoneName=root, Default=1, Name=deformRig_r_thigh_up_alpha) {template="DISPATCH_RigDispatch_GetAnimAttribute(in Name,in BoneName,in Default,out Value,out Found)"}.Value
   )
   n1 = RigVMDispatchNode(BoneName=root, Default=0, Name=deformRig_r_upperarm_up_alpha) {template="DISPATCH_RigDispatch_GetAnimAttribute(in Name,in BoneName,in Default,out Value,out Found)"}
   > &RigUnit_ModifyTransforms(
    ItemToModify="((Item=(Type=Bone,Name=\"deform_upperarm_bk_r\"),Transform=(Rotation=(X=0,Y=0,Z=0,W=1),Translation=(X=-5.5,Y=2,Z=3),Scale3D=(X=1,Y=1,Z=1))),(Item=(Type=Bone,Name=\"deform_lat_r\"),Transform=(Rotation=(X=0,Y=0,Z=0,W=1),Translation=(X=0,Y=0,Z=10),Scale3D=(X=1,Y= ...(truncated)"
    WeightMinimum=0
    Weight=@n1.Value
   )
   > &RigUnit_ModifyTransforms(
    ItemToModify=((Item=(Type=Bone,Name="deform_lat_r"),Transform=(Rotation=(X=0,Y=0,Z=0,W=1),Translation=(X=0,Y=0,Z=-20),Scale3D=(X=1,Y=1,Z=1))),(Item=(Type=Bone,Name="deform_lat_l"),Transform=(Rotation=(X=0,Y=0,Z=0,W=1),Translation=(X=0,Y=0,Z=20),Scale3D=(X=1,Y=1,Z=1))))
    WeightMinimum=0
    Weight=&RigVMFunction_MathDoubleMul(A=@n1.Value, B=RigVMVariableNode() {variable=ParachuteOpen}.Value).Result
   )
   > &RigUnit_ModifyTransforms(
    ItemToModify="((Item=(Type=Bone,Name=\"base_shortBck_01_l\"),Transform=(Rotation=(X=0,Y=0,Z=0,W=1),Translation=(X=-2,Y=0,Z=0),Scale3D=(X=1,Y=1,Z=1))),(Item=(Type=Bone,Name=\"base_shortOut_01_l\"),Transform=(Rotation=(X=0,Y=0,Z=-0.0872,W=0.9962),Translation=(X=-2,Y=0,Z=0),Sc ...(truncated)"
    WeightMinimum=0
    Weight=&RigVMFunction_MathDoubleMul(
     A=RigVMDispatchNode(BoneName=root, Default=1, Name=deformRig_l_knee_alpha) {template="DISPATCH_RigDispatch_GetAnimAttribute(in Name,in BoneName,in Default,out Value,out Found)"}.Value
     B=RigVMVariableNode() {variable=crouchWalking}.Value
    ).Result
   )
   Multiply = RigVMAggregateNode(
    A=RigVMDispatchNode(BoneName=root, Default=1, Name=deformRig_r_knee_alpha) {template="DISPATCH_RigDispatch_GetAnimAttribute(in Name,in BoneName,in Default,out Value,out Found)"}.Value
    B=RigVMVariableNode() {variable=crouchWalking}.Value
    C=RigVMVariableNode() {variable=isDBNO}.Value
   )
   > &RigUnit_ModifyTransforms(
    ItemToModify="((Item=(Type=Bone,Name=\"base_shortBck_01_r\"),Transform=(Rotation=(X=0,Y=0.4226,Z=0,W=0.9063),Translation=(X=-10,Y=0,Z=0),Scale3D=(X=1,Y=1,Z=1))),(Item=(Type=Bone,Name=\"base_shortOut_01_r\"),Transform=(Rotation=(X=0,Y=0,Z=0.2588,W=0.9659),Translation=(X=-5,Y ...(truncated)"
    WeightMinimum=0
    Weight=@Multiply.Result
   )
   > &RigUnit_ModifyTransforms(
    ItemToModify="((Item=(Type=Bone,Name=\"deform_upperarm_up_l\"),Transform=(Rotation=(X=0,Y=0,Z=0,W=1),Translation=(X=2.3719,Y=0,Z=-0.5437),Scale3D=(X=1,Y=1,Z=1))),(Item=(Type=Bone,Name=\"deform_upperarm_up_r\"),Transform=(Rotation=(X=0,Y=0,Z=0,W=1),Translation=(X=1.8289,Y=0, ...(truncated)"
    WeightMinimum=0
    Weight=RigVMVariableNode() {variable=ParachuteOpen}.Value
   )
   > &RigUnit_ModifyTransforms(
    ItemToModify="((Item=(Type=Bone,Name=\"dyn_chestHelper_01_l\"),Transform=(Rotation=(X=0,Y=0,Z=0,W=1),Translation=(X=0,Y=8.6174,Z=0),Scale3D=(X=1,Y=1,Z=1))),(Item=(Type=Bone,Name=\"dyn_necklace_01_r\"),Transform=(Rotation=(X=0,Y=0,Z=-0.342,W=-0.9397),Translation=(X=0,Y=0,Z=0 ...(truncated)"
    WeightMinimum=0
    Weight=RigVMVariableNode() {variable=ParachuteOpen}.Value
   )
   n2 = &RigUnit_SphericalPoseReader(
    ActiveRegionScaleFactors=(PositiveWidth=0.5,NegativeWidth=1,PositiveHeight=0.5,NegativeHeight=0.5)
    ActiveRegionSize=0.1
    Debug=(bDrawDebug=false,bDraw2D=False,bDrawLocalAxes=False,DebugScale=25,DebugSegments=20,DebugThickness=0.25)
    DriverAxis=(X=-1,Y=0,Z=0)
    DriverItem=(Type=Bone,Name="hand_r")
    FalloffRegionScaleFactors=(PositiveWidth=0.2,NegativeWidth=1,PositiveHeight=0.5,NegativeHeight=0.5)
    FalloffSize=0.2
    FlipHeightScaling=False
    FlipWidthScaling=False
    OptionalParentItem=(Type=Bone,Name="None")
    RotationOffset=(X=140,Y=0,Z=30)
    ExecutePin=@RigVMFunction_Sequence.C
   )
   n3 = &RigVMFunction_MathDoubleMul(A=@n2.OutputParam, B=RigVMVariableNode() {variable=largeGunEquipped}.Value)
   n4 = RigVMVariableNode() {variable=ShotGunArm}
   > &RigUnit_ModifyTransforms(
    ItemToModify="((Item=(Type=Bone,Name=\"lowerarm_twist_02_l\"),Transform=(Rotation=(X=0.4226,Y=0,Z=0,W=0.9063),Translation=(X=0,Y=0,Z=0),Scale3D=(X=1,Y=1.1,Z=1.1))),(Item=(Type=Bone,Name=\"lowerarm_twist_01_l\"),Transform=(Rotation=(X=-0.1736,Y=0,Z=0,W=0.9848),Translation=(X ...(truncated)"
    WeightMinimum=0
    Weight=&RigVMFunction_MathDoubleRemap(
     SourceMaximum=1
     SourceMinimum=0
     TargetMaximum=1
     TargetMinimum=0
     bClamp=False
     Value=&RigVMFunction_MathDoubleMul(A=@n3.Result, B=@n4.Value).Result
    ).Result
   )
   > &RigUnit_ModifyTransforms(
    ItemToModify=((Item=(Type=Bone,Name="base_watch_02_r"),Transform=(Rotation=(X=0,Y=0,Z=0,W=1),Translation=(X=7.414,Y=-4.2919,Z=-0.0229),Scale3D=(X=1,Y=1,Z=1))))
    WeightMinimum=0
    Weight=@n2.OutputParam
   )
   > &RigUnit_ModifyTransforms(
    ItemToModify="((Item=(Type=Bone,Name=\"base_watch_01_r\"),Transform=(Rotation=(X=0,Y=0,Z=0,W=1),Translation=(X=7.4641,Y=0,Z=3.6783),Scale3D=(X=1,Y=1,Z=1))),(Item=(Type=Bone,Name=\"base_watch_02_r\"),Transform=(Rotation=(X=-0.169,Y=0,Z=0,W=0.9856),Translation=(X=0,Y=-9.2271, ...(truncated)"
    WeightMinimum=0
    Weight=RigVMVariableNode() {variable=SkydiveMovement}.Value
   )
   > n5 = &RigUnit_SphericalPoseReader(
    ActiveRegionScaleFactors=(PositiveWidth=0.6,NegativeWidth=1,PositiveHeight=0.6,NegativeHeight=1)
    ActiveRegionSize=0.2
    Debug=(bDrawDebug=true,bDraw2D=False,bDrawLocalAxes=False,DebugScale=25,DebugSegments=20,DebugThickness=0.25)
    DriverAxis=(X=1,Y=0,Z=0)
    DriverItem=(Type=Bone,Name="upperarm_l")
    FalloffRegionScaleFactors=(PositiveWidth=0.5,NegativeWidth=1,PositiveHeight=1.2,NegativeHeight=1)
    FalloffSize=0.4
    FlipHeightScaling=False
    FlipWidthScaling=False
    OptionalParentItem=(Type=Bone,Name="None")
    RotationOffset=(X=3.2146,Y=-46.3808,Z=0)
   )
   > &RigUnit_ModifyTransforms(
    ItemToModify="((Item=(Type=Bone,Name=\"dyn_necklace_01_r\"),Transform=(Rotation=(X=-0.1633,Y=-0.0969,Z=0.0957,W=0.9771),Translation=(X=-3,Y=2,Z=6),Scale3D=(X=1,Y=1.2269,Z=1))),(Item=(Type=Bone,Name=\"base_chestHelper_01_r\"),Transform=(Rotation=(X=0,Y=0,Z=0,W=1),Translation ...(truncated)"
    WeightMinimum=0
    Weight=@n5.OutputParam
   )
   > &RigUnit_ModifyTransforms(
    ItemToModify="((Item=(Type=Bone,Name=\"base_necklace_01_r\"),Transform=(Rotation=(X=0,Y=0,Z=0,W=1),Translation=(X=-5,Y=0,Z=0),Scale3D=(X=1,Y=1,Z=1))),(Item=(Type=Bone,Name=\"base_necklace_01_l\"),Transform=(Rotation=(X=0,Y=0,Z=0,W=1),Translation=(X=-5,Y=0,Z=-4),Scale3D=(X=1 ...(truncated)"
    WeightMinimum=0
    Weight=RigVMVariableNode() {variable=SkydiveMovement}.Value
   )
   > &RigUnit_ModifyTransforms(
    ItemToModify="((Item=(Type=Bone,Name=\"base_necklace_01_r\"),Transform=(Rotation=(X=0.171,Y=-0.171,Z=0.0302,W=0.9698),Translation=(X=-4,Y=5,Z=-4),Scale3D=(X=1,Y=1,Z=1))),(Item=(Type=Bone,Name=\"base_necklace_01_l\"),Transform=(Rotation=(X=0,Y=0,Z=0,W=1),Translation=(X=-10,Y ...(truncated)"
    WeightMinimum=0
    Weight=RigVMVariableNode() {variable=ParachuteOpen}.Value
   )
   RigVMFunction_MathBoolToFloat = &RigVMFunction_MathBoolToFloat(Value=RigVMVariableNode() {variable=LeftMove}.Value)
   > &RigUnit_ModifyTransforms(
    ItemToModify="((Item=(Type=Bone,Name=\"base_shortInner_01_l\"),Transform=(Rotation=(X=0,Y=-0.0872,Z=0,W=0.9962),Translation=(X=0,Y=0,Z=1.5),Scale3D=(X=1,Y=1,Z=1))),(Item=(Type=Bone,Name=\"base_shortInner_01_r\"),Transform=(Rotation=(X=0,Y=0,Z=0,W=1),Translation=(X=0,Y=0,Z=3 ...(truncated)"
    WeightMinimum=0
    Weight=@RigVMFunction_MathBoolToFloat.Result
   )
   > &RigUnit_ModifyTransforms(
    ItemToModify="((Item=(Type=Bone,Name=\"base_shortInner_01_l\"),Transform=(Rotation=(X=0,Y=-0.0872,Z=0,W=0.9962),Translation=(X=0,Y=0,Z=1.5),Scale3D=(X=1,Y=1,Z=1))),(Item=(Type=Bone,Name=\"base_shortInner_01_r\"),Transform=(Rotation=(X=0,Y=0,Z=0,W=1),Translation=(X=0,Y=0,Z=3 ...(truncated)"
    WeightMinimum=0
    Weight=&RigVMFunction_MathBoolToFloat(Value=RigVMVariableNode() {variable=RightMove}.Value).Result
   )
   > n6 = &RigUnit_SphericalPoseReader(
    ActiveRegionScaleFactors=(PositiveWidth=1,NegativeWidth=0.5,PositiveHeight=0.75,NegativeHeight=0.75)
    ActiveRegionSize=0.1
    Debug=(bDrawDebug=true,bDraw2D=False,bDrawLocalAxes=False,DebugScale=25,DebugSegments=20,DebugThickness=0.25)
    DriverAxis=(X=0,Y=1,Z=0)
    DriverItem=(Type=Bone,Name="head")
    FalloffRegionScaleFactors=(PositiveWidth=1,NegativeWidth=0.5,PositiveHeight=1,NegativeHeight=1)
    FalloffSize=0.2
    FlipHeightScaling=False
    FlipWidthScaling=False
    OptionalParentItem=(Type=Bone,Name="None")
    RotationOffset=(X=90,Y=0,Z=-30)
   )
   > n7 = &RigUnit_SphericalPoseReader(
    ActiveRegionScaleFactors=(PositiveWidth=1,NegativeWidth=1,PositiveHeight=1,NegativeHeight=1)
    ActiveRegionSize=0.025
    Debug=(bDrawDebug=true,bDraw2D=False,bDrawLocalAxes=False,DebugScale=25,DebugSegments=20,DebugThickness=0.25)
    DriverAxis=(X=-1,Y=0,Z=0)
    DriverItem=(Type=Bone,Name="upperarm_r")
    FalloffRegionScaleFactors=(PositiveWidth=1,NegativeWidth=1,PositiveHeight=1,NegativeHeight=1)
    FalloffSize=0.1
    FlipHeightScaling=False
    FlipWidthScaling=False
    OptionalParentItem=(Type=Bone,Name="None")
    RotationOffset=(X=-115,Y=0,Z=25)
   )
   > &RigUnit_ModifyTransforms(
    ItemToModify="((Item=(Type=Bone,Name=\"None\"),Transform=(Rotation=(X=0,Y=0,Z=0,W=1),Translation=(X=0,Y=0,Z=0),Scale3D=(X=1,Y=1,Z=1))),(Item=(Type=Bone,Name=\"dyn_necklace_01_r\"),Transform=(Rotation=(X=0,Y=0,Z=0,W=1),Translation=(X=-0.1401,Y=0,Z=2.5755),Scale3D=(X=1,Y=1,Z= ...(truncated)"
    WeightMinimum=0
    Weight=&RigVMFunction_MathDoubleMul(A=@n7.OutputParam, B=@n6.OutputParam).Result
   )
   &RigUnit_ModifyTransforms(
    ItemToModify=((Item=(Type=Bone,Name="base_GasCan_01_l"),Transform=(Rotation=(X=0.422,Y=0.0294,Z=-0.007,W=0.9061),Translation=(X=-0.8353,Y=-1.2556,Z=1.4364),Scale3D=(X=1,Y=1,Z=1))))
    WeightMinimum=0
    Weight=RigVMVariableNode() {variable=SkydiveMovement}.Value
    ExecutePin=@RigVMFunction_Sequence.A
   )
   > &RigUnit_ParentConstraint(
    AdvancedSettings=(RotationOrderForFilter=XZY)
    Child=(Type=Bone,Name="base_megaphone_01_r")
    Parents="((Item=(Type=Bone,Name=\"deform_groin_l\"),Weight=0.0001),(Item=(Type=Bone,Name=\"thigh_twist_01_l\"),Weight=0.0002),(Item=(Type=Bone,Name=\"spine_04\"),Weight=0.0003),(Item=(Type=Bone,Name=\"deform_glute_r\"),Weight=0.0053),(Item=(Type=Bone,Name=\"spine_03\"),Weigh ...(truncated)"
    Weight=1
   )
   > &RigUnit_ParentConstraint(
    AdvancedSettings=(RotationOrderForFilter=XZY)
    Child=(Type=Bone,Name="base_armRope_01_l")
    Parents=((Item=(Type=Bone,Name="base_GasCan_01_l"),Weight=1.0))
    Weight=1
   )
   > &RigUnit_ModifyTransforms(
    ItemToModify="((Item=(Type=Bone,Name=\"base_megaphone_01_r\"),Transform=(Rotation=(X=0,Y=-0.0523,Z=0,W=0.9986),Translation=(X=0,Y=0,Z=0),Scale3D=(X=1,Y=1,Z=1))),(Item=(Type=Bone,Name=\"dyn_necklace_01_r\"),Transform=(Rotation=(X=0,Y=0.1764,Z=0,W=0.9843),Translation=(X=0,Y=- ...(truncated)"
    Weight=1
    WeightMinimum=0
   )
   n8 = RigVMVariableNode() {variable=RailGrind}
   > &RigUnit_RotationConstraintLocalSpaceOffset(
    AdvancedSettings=(InterpolationType=Average,RotationOrderForFilter=XZY)
    Child=(Type=Bone,Name="base_megaphone_01_r")
    Filter=(bX=true,bY=true,bZ=true)
    Parents=((Item=(Type=Bone,Name="thigh_r"),Weight=1))
    Weight=@n8.Value
   )
   > &RigUnit_ModifyTransforms(
    ItemToModify=((Item=(Type=Bone,Name="base_shortBck_01_l"),Transform=(Rotation=(X=0,Y=0,Z=0,W=1),Translation=(X=0,Y=-10,Z=0),Scale3D=(X=1,Y=1,Z=1))))
    WeightMinimum=0
    Weight=@n8.Value
   )
   > &RigUnit_ParentConstraint(
    AdvancedSettings=(RotationOrderForFilter=XZY)
    Child=(Type=Bone,Name="spine_05_base_collider")
    Parents="((Item=(Type=Bone,Name=\"spine_01\"),Weight=0.0078),(Item=(Type=Bone,Name=\"clavicle_r\"),Weight=0.0099),(Item=(Type=Bone,Name=\"deform_pec_r\"),Weight=0.0163),(Item=(Type=Bone,Name=\"deform_pec_l\"),Weight=0.0181),(Item=(Type=Bone,Name=\"spine_02\"),Weight=0.1114), ...(truncated)"
    Weight=1
   )
   > &RigUnit_ParentConstraint(
    AdvancedSettings=(RotationOrderForFilter=XZY)
    Child=(Type=Bone,Name="base_necklace_01_r")
    Parents="((Item=(Type=Bone,Name=\"spine_03\"),Weight=0.0086),(Item=(Type=Bone,Name=\"neck_01\"),Weight=0.0557),(Item=(Type=Bone,Name=\"head\"),Weight=0.1154),(Item=(Type=Bone,Name=\"neck_02\"),Weight=0.1929),(Item=(Type=Bone,Name=\"clavicle_r\"),Weight=0.2306),(Item=(Type=Bo ...(truncated)"
    Weight=1
   )
   > &RigUnit_ParentConstraint(
    AdvancedSettings=(RotationOrderForFilter=XZY)
    Child=(Type=Bone,Name="base_necklaceCol_01_l")
    Parents="((Item=(Type=Bone,Name=\"spine_03\"),Weight=0.0325),(Item=(Type=Bone,Name=\"upperarm_twist_01_l\"),Weight=0.0483),(Item=(Type=Bone,Name=\"deform_upperarm_up_l\"),Weight=0.0555),(Item=(Type=Bone,Name=\"deform_upperarm_fr_l\"),Weight=0.0609),(Item=(Type=Bone,Name=\"s ...(truncated)"
    Weight=1
   )
   > &RigUnit_ParentConstraint(
    AdvancedSettings=(RotationOrderForFilter=XZY)
    Child=(Type=Bone,Name="base_necklaceCol_01_r")
    Parents="((Item=(Type=Bone,Name=\"spine_03\"),Weight=0.0325),(Item=(Type=Bone,Name=\"upperarm_twist_01_r\"),Weight=0.0483),(Item=(Type=Bone,Name=\"deform_upperarm_up_r\"),Weight=0.0555),(Item=(Type=Bone,Name=\"deform_upperarm_fr_r\"),Weight=0.0609),(Item=(Type=Bone,Name=\"s ...(truncated)"
    Weight=1
   )
   > &RigUnit_ParentConstraint(
    AdvancedSettings=(RotationOrderForFilter=XZY)
    Child=(Type=Bone,Name="base_chestHelper_01_l")
    Parents="((Item=(Type=Bone,Name=\"spine_03\"),Weight=0.0175),(Item=(Type=Bone,Name=\"upperarm_twist_01_l\"),Weight=0.0234),(Item=(Type=Bone,Name=\"deform_upperarm_up_l\"),Weight=0.0267),(Item=(Type=Bone,Name=\"deform_upperarm_fr_l\"),Weight=0.0293),(Item=(Type=Bone,Name=\"d ...(truncated)"
    Weight=1
   )
   > &RigUnit_ParentConstraint(
    AdvancedSettings=(RotationOrderForFilter=XZY)
    Child=(Type=Bone,Name="base_chestHelper_01_r")
    Parents="((Item=(Type=Bone,Name=\"spine_03\"),Weight=0.0175),(Item=(Type=Bone,Name=\"upperarm_twist_01_r\"),Weight=0.0234),(Item=(Type=Bone,Name=\"deform_upperarm_up_r\"),Weight=0.0267),(Item=(Type=Bone,Name=\"deform_upperarm_fr_r\"),Weight=0.0293),(Item=(Type=Bone,Name=\"d ...(truncated)"
    Weight=1
   )
   > &RigUnit_ParentConstraint(
    AdvancedSettings=(RotationOrderForFilter=XZY)
    Child=(Type=Bone,Name="base_watch_02_r")
    Parents="((Item=(Type=Bone,Name=\"thumb_01_r\"),Weight=0.0017),(Item=(Type=Bone,Name=\"deform_wrist_dn_r\"),Weight=0.0222),(Item=(Type=Bone,Name=\"lowerarm_r\"),Weight=0.0304),(Item=(Type=Bone,Name=\"deform_wrist_up_r\"),Weight=0.1002),(Item=(Type=Bone,Name=\"pinky_metacarp ...(truncated)"
    Weight=1
   )
   > &RigUnit_ParentConstraint(
    AdvancedSettings=(RotationOrderForFilter=XZY)
    Child=(Type=Bone,Name="base_necklace_01_bk")
    Parents="((Item=(Type=Bone,Name=\"spine_04\"),Weight=0.0019),(Item=(Type=Bone,Name=\"spine_03\"),Weight=0.0031),(Item=(Type=Bone,Name=\"clavicle_r\"),Weight=0.0301),(Item=(Type=Bone,Name=\"clavicle_l\"),Weight=0.0314),(Item=(Type=Bone,Name=\"neck_02\"),Weight=0.1032),(Item=( ...(truncated)"
    Weight=1
   )
   > &RigUnit_ParentConstraint(
    AdvancedSettings=(RotationOrderForFilter=XZY)
    Child=(Type=Bone,Name="dyn_necklace_01_bk_AIM")
    Parents="((Item=(Type=Bone,Name=\"spine_04\"),Weight=0.0039),(Item=(Type=Bone,Name=\"spine_03\"),Weight=0.0072),(Item=(Type=Bone,Name=\"clavicle_l\"),Weight=0.0079),(Item=(Type=Bone,Name=\"clavicle_r\"),Weight=0.0836),(Item=(Type=Bone,Name=\"neck_02\"),Weight=0.1159),(Item=( ...(truncated)"
    Weight=1
   )
   > &RigUnit_AimConstraintLocalSpaceOffset(
    AimAxis=(X=0,Y=1,Z=0)
    Child=(Type=Bone,Name="dyn_necklace_01_bk")
    Filter=(bX=True,bY=True,bZ=True)
    Parents=((Item=(Type=Bone,Name="dyn_necklace_01_bk_AIM"),Weight=1))
    UpAxis=(X=-1,Y=0,Z=0)
    Weight=1
    WorldUp=(Target=(X=0,Y=0,Z=1),Kind=Location,Space=(Type=Bone,Name="dyn_necklace_01_bk_AIM"))
   )
   &RigUnit_ModifyTransforms(
    ExecutePin=()
    ItemToModify="((Item=(Type=Bone,Name=\"base_necklace_01_bk\"),Transform=(Rotation=(X=0,Y=0,Z=0,W=1),Translation=(X=0,Y=0,Z=20),Scale3D=(X=1,Y=1,Z=1))),(Item=(Type=Bone,Name=\"base_necklace_01_l\"),Transform=(Rotation=(X=0.2588,Y=0,Z=0,W=0.9659),Translation=(X=3,Y=0,Z=1),Sca ...(truncated)"
    Weight=1
    WeightMinimum=0
   )
   &RigVMFunction_MathDoubleMul(A=@n5.OutputParam, B=RigVMVariableNode() {variable=pistolHand}.Value)
  ]}
]
hierarchy [
 root {loc=<0 0 0>, rot=<0 0 0 1>}
 pelvis {parent=root, loc=<0 -0.7321 92.4218>, rot=<0.0127 0.707 -0.0127 -0.707>}
 spine_01 {parent=pelvis, loc=<5.6002 -0.5128 0>, rot=<0 0 0.034 -0.9994>}
 spine_02 {parent=spine_01, loc=<8.0753 0.0139 0>, rot=<0 0 0.0238 -0.9997>}
 spine_03 {parent=spine_02, loc=<7.0393 0 0>, rot=<0 0 0.0149 -0.9999>}
 spine_04 {parent=spine_03, loc=<8.1024 0 0>, rot=<0 0 -0.0132 -0.9999>}
 spine_05 {parent=spine_04, loc=<10.0146 0 0>, rot=<0 0 -0.0015 -1>}
 attach_cape {parent=spine_05, loc=<9.017 -11.7141 0>, rot=<0.0616 -0.7044 -0.0616 -0.7044>}
 attach_backpack {parent=spine_05, loc=<-0.123 -13.473 0>, rot=<-0.0093 -0.707 0.0093 -0.707>}
 neck_01 {parent=spine_05, loc=<10.815 1.9196 0>, rot=<0 0 -0.1864 -0.9825>}
 neck_02 {parent=neck_01, loc=<4.7347 0.3045 0>, rot=<0 0 0.0363 -0.9993>}
 head {parent=neck_02, loc=<5.7632 0.5552 0>, rot=<0 0 0.115 -0.9934>}
 base_necklace_01_r {parent=neck_02, loc=<0.282 -12.533 10.375>, rot=<-0.0784 -0.7027 0.0784 -0.7027>}
 dyn_necklace_01_r {parent=base_necklace_01_r, loc=<0 0 0>, rot=<0 0 0 1>}
 base_necklace_01_bk {parent=neck_01, loc=<3.3146 -16.1115 0>, rot=<0.0432 -0.7058 0.7058 -0.0432>}
 dyn_necklace_01_bk {parent=base_necklace_01_bk, loc=<0 0 0>, rot=<0 0 0 1>}
 base_necklace_01_l {parent=neck_01, loc=<4.2686 -7.1654 -13.6525>, rot=<-0.1038 -0.6994 0.1038 -0.6994>}
 dyn_necklace_01_l {parent=base_necklace_01_l, loc=<0 0 0>, rot=<0 0 0 1>}
 dyn_necklace_01_bk_AIM {parent=neck_01, loc=<3.3146 -16.1115 3.3>, rot=<-0.0432 0.7058 -0.7058 0.0432>}
 clavicle_l {parent=spine_05, loc=<2.5856 -0.2866 -4.3777>, rot=<0.0465 -0.7289 -0.009 -0.6829>}
 upperarm_l {parent=clavicle_l, loc=<18.2264 0.1342 -0.001>, rot=<0.0253 -0.2693 0.0013 -0.9627>}
 lowerarm_l {parent=upperarm_l, loc=<29.0071 0.1686 -1.239>, rot=<0.0391 -0.0099 -0.2453 -0.9686>}
 hand_l {parent=lowerarm_l, loc=<25.8607 0.222 1.0274>, rot=<-0.7193 -0.0988 0.1193 -0.6773>}
 thumb_01_l {parent=hand_l, loc=<1.3118 -0.3191 -2.459>, rot=<0.6546 -0.4042 0.0664 -0.6354>}
 thumb_02_l {parent=thumb_01_l, loc=<5.9336 -0.0448 0.0401>, rot=<-0.0037 0.0213 0.298 -0.9543>}
 thumb_03_l {parent=thumb_02_l, loc=<2.2223 0.3682 -0.0907>, rot=<0.0118 -0.013 0.0066 -0.9998>}
 index_metacarpal_l {parent=hand_l, loc=<1.6535 0.0183 -2.96>, rot=<0.0158 -0.0124 -0.0173 -0.9996>}
 index_01_l {parent=index_metacarpal_l, loc=<8.3008 0.0357 -0.0524>, rot=<-0.0044 0.0485 0.0924 -0.9945>}
 index_02_l {parent=index_01_l, loc=<3.3587 -0.3004 0.0452>, rot=<-0.002 0.0027 0.3225 -0.9466>}
 index_03_l {parent=index_02_l, loc=<2.5578 -0.123 -0.0252>, rot=<-0.002 0.0067 0.2558 -0.9667>}
 middle_metacarpal_l {parent=hand_l, loc=<2.1856 -0.1725 -0.5984>, rot=<0.006 0.0168 -0.0329 -0.9993>}
 middle_01_l {parent=middle_metacarpal_l, loc=<7.3275 -0.2447 0.0129>, rot=<-0.0208 0.0042 0.0858 -0.9961>}
 middle_02_l {parent=middle_01_l, loc=<4.0123 0.1401 0.1175>, rot=<-0.0218 0.0258 0.3512 -0.9357>}
 middle_03_l {parent=middle_02_l, loc=<2.6077 0.0534 0.1373>, rot=<0.0082 0.0188 0.1708 -0.9851>}
 ring_metacarpal_l {parent=hand_l, loc=<2.2074 -0.3625 1.3638>, rot=<-0.0284 0.0722 -0.0458 -0.9959>}
 ring_01_l {parent=ring_metacarpal_l, loc=<7.0295 -0.6197 -0.1446>, rot=<-0.0489 -0.0252 0.1892 -0.9804>}
 ring_02_l {parent=ring_01_l, loc=<3.6174 0.4647 -0.1027>, rot=<0 0 0.2806 -0.9598>}
 ring_03_l {parent=ring_02_l, loc=<2.4152 0.114 0.0199>, rot=<0 0 0.1189 -0.9929>}
 pinky_metacarpal_l {parent=hand_l, loc=<1.7616 -0.9753 3.3814>, rot=<-0.0397 0.0814 -0.069 -0.9935>}
 pinky_01_l {parent=pinky_metacarpal_l, loc=<7.1709 -0.4831 -0.1094>, rot=<-0.0654 -0.0035 0.2242 -0.9723>}
 pinky_02_l {parent=pinky_01_l, loc=<2.8274 0.0893 -0.1043>, rot=<0.0037 0.0011 0.2854 -0.9584>}
 pinky_03_l {parent=pinky_02_l, loc=<2.2046 0.0307 -0.0078>, rot=<-0.0002 0.0098 0.1445 -0.9895>}
 deform_wrist_dn_base_l {parent=hand_l, loc=<0.0004 -3.4171 0>, rot=<0 0 0 -1>}
 deform_wrist_dn_l {parent=deform_wrist_dn_base_l, loc=<0 0 0>, rot=<0 0 0 1>}
 deform_wrist_up_base_l {parent=hand_l, loc=<0.0005 2.7807 0>, rot=<0 0 0 -1>}
 deform_wrist_up_l {parent=deform_wrist_up_base_l, loc=<0 0 0>, rot=<0 0 0 1>}
 lowerarm_twist_02_l {parent=lowerarm_l, loc=<9.6199 0 0>, rot=<0 0 0 -1>}
 lowerarm_twist_01_l {parent=lowerarm_l, loc=<17.8549 0 0>, rot=<0 0 0 -1>, scale=<1.0001 1 1>}
 deform_elbow_in_base_l {parent=lowerarm_l, loc=<0.2112 5.645 0.8274>, rot=<0 0 0 -1>}
 deform_elbow_in_l {parent=deform_elbow_in_base_l, loc=<0 0 0>, rot=<0 0 0 1>}
 deform_elbow_out_base_l {parent=lowerarm_l, loc=<0 -7.5129 0>, rot=<0 0 0 -1>}
 deform_elbow_out_l {parent=deform_elbow_out_base_l, loc=<0 0 0>, rot=<0 0 0 1>}
 upperarm_twist_01_l {parent=upperarm_l, loc=<7.9796 0.0843 -0.6195>, rot=<0 0 0 -1>}
 deform_upperarm_bk_base_l {parent=upperarm_twist_01_l, loc=<-6.8983 -13.3826 -1.3258>, rot=<0 0 0 1>}
 deform_upperarm_bk_l {parent=deform_upperarm_bk_base_l, loc=<0 0 0>, rot=<0 0 0 1>}
 deform_upperarm_fr_base_l {parent=upperarm_twist_01_l, loc=<-8.3487 10.7117 -2.1455>, rot=<0 0 0 1>}
 deform_upperarm_fr_l {parent=deform_upperarm_fr_base_l, loc=<0 0 0>, rot=<0 0 0 1>}
 deform_upperarm_up_base_l {parent=upperarm_twist_01_l, loc=<-11.2012 -1.2234 8.7784>, rot=<0 0 0 1>}
 deform_upperarm_up_l {parent=deform_upperarm_up_base_l, loc=<0 0 0>, rot=<0 0 0 1>}
 base_armRope_01_l {parent=upperarm_twist_01_l, loc=<3.554 6.3049 8.3116>, rot=<0.5349 -0.337 -0.3091 -0.7105>}
 dyn_armRope_01_l {parent=base_armRope_01_l, loc=<0 0 0>, rot=<0 0 0 1>}
 dyn_armRope_02_l {parent=dyn_armRope_01_l, loc=<1.7721 0 0>, rot=<0 -0.0033 -0.0012 -1>}
 dyn_armRope_03_l {parent=dyn_armRope_02_l, loc=<1.8129 0 0>, rot=<0 0 0 1>}
 upperarm_twist_02_l {parent=upperarm_l, loc=<19.5059 0.0843 -0.6195>, rot=<0 0 0 -1>}
 deform_bicep_base_l {parent=upperarm_twist_02_l, loc=<0 8.0494 0>, rot=<0 0 0 1>}
 deform_bicep_l {parent=deform_bicep_base_l, loc=<0 0 0>, rot=<0 0 0 1>}
 base_GasCan_01_l {parent=upperarm_twist_02_l, loc=<0 0 0>, rot=<0 0 0 1>}
 clavicle_r {parent=spine_05, loc=<2.5855 -0.2866 4.3777>, rot=<0.7289 0.0465 -0.6829 0.009>}
 upperarm_r {parent=clavicle_r, loc=<-18.2264 -0.1342 0.001>, rot=<0.0253 -0.2693 0.0013 -0.9627>}
 lowerarm_r {parent=upperarm_r, loc=<-29.0071 -0.1686 1.239>, rot=<0.0391 -0.0099 -0.2453 -0.9686>}
 hand_r {parent=lowerarm_r, loc=<-25.8607 -0.222 -1.0274>, rot=<-0.7193 -0.0988 0.1193 -0.6773>}
 thumb_01_r {parent=hand_r, loc=<-1.3118 0.3191 2.459>, rot=<0.6546 -0.4042 0.0664 -0.6354>}
 thumb_02_r {parent=thumb_01_r, loc=<-5.9336 0.0448 -0.0401>, rot=<-0.0037 0.0213 0.298 -0.9543>}
 thumb_03_r {parent=thumb_02_r, loc=<-2.2223 -0.3682 0.0907>, rot=<0.0118 -0.013 0.0066 -0.9998>}
 index_metacarpal_r {parent=hand_r, loc=<-1.6535 -0.0183 2.96>, rot=<0.0158 -0.0124 -0.0173 -0.9996>}
 index_01_r {parent=index_metacarpal_r, loc=<-8.3008 -0.0357 0.0524>, rot=<-0.0044 0.0485 0.0924 -0.9945>}
 index_02_r {parent=index_01_r, loc=<-3.3587 0.3004 -0.0452>, rot=<-0.002 0.0027 0.3225 -0.9466>}
 index_03_r {parent=index_02_r, loc=<-2.5578 0.123 0.0252>, rot=<-0.002 0.0067 0.2558 -0.9667>}
 middle_metacarpal_r {parent=hand_r, loc=<-2.1856 0.1725 0.5984>, rot=<0.006 0.0168 -0.0329 -0.9993>}
 middle_01_r {parent=middle_metacarpal_r, loc=<-7.3275 0.2447 -0.0129>, rot=<-0.0208 0.0042 0.0858 -0.9961>}
 middle_02_r {parent=middle_01_r, loc=<-4.0123 -0.1401 -0.1175>, rot=<-0.0218 0.0258 0.3512 -0.9357>}
 middle_03_r {parent=middle_02_r, loc=<-2.6077 -0.0534 -0.1373>, rot=<0.0082 0.0188 0.1708 -0.9851>}
 ring_metacarpal_r {parent=hand_r, loc=<-2.2074 0.3625 -1.3638>, rot=<-0.0284 0.0722 -0.0458 -0.9959>}
 ring_01_r {parent=ring_metacarpal_r, loc=<-7.0295 0.6197 0.1446>, rot=<-0.0489 -0.0252 0.1892 -0.9804>}
 ring_02_r {parent=ring_01_r, loc=<-3.6174 -0.4647 0.1027>, rot=<0 0 0.2806 -0.9598>}
 ring_03_r {parent=ring_02_r, loc=<-2.4152 -0.114 -0.0199>, rot=<0 0 0.1189 -0.9929>}
 pinky_metacarpal_r {parent=hand_r, loc=<-1.7616 0.9753 -3.3814>, rot=<-0.0397 0.0814 -0.069 -0.9935>}
 pinky_01_r {parent=pinky_metacarpal_r, loc=<-7.1709 0.4831 0.1094>, rot=<-0.0654 -0.0035 0.2242 -0.9723>}
 pinky_02_r {parent=pinky_01_r, loc=<-2.8274 -0.0893 0.1043>, rot=<0.0037 0.0011 0.2854 -0.9584>}
 pinky_03_r {parent=pinky_02_r, loc=<-2.2046 -0.0307 0.0078>, rot=<-0.0002 0.0098 0.1445 -0.9895>}
 deform_wrist_dn_base_r {parent=hand_r, loc=<0 3.4172 0>, rot=<0 0 0 -1>}
 deform_wrist_dn_r {parent=deform_wrist_dn_base_r, loc=<0 0 0>, rot=<0 0 0 -1>}
 deform_wrist_up_base_r {parent=hand_r, loc=<-0.0001 -2.7804 0>, rot=<0 0 0 -1>}
 deform_wrist_up_r {parent=deform_wrist_up_base_r, loc=<0 0 0>, rot=<0 0 0 -1>}
 base_watch_01_r {parent=hand_r, loc=<2.188 -2.5382 -0.0633>, rot=<-0.5251 0.463 0.4997 -0.5101>}
 base_watch_02_r {parent=hand_r, loc=<1.8236 0.3783 -4.6857>, rot=<0.5498 -0.4634 -0.5338 -0.445>}
 lowerarm_twist_02_r {parent=lowerarm_r, loc=<-9.6199 0 0>, rot=<0 0 0 -1>}
 lowerarm_twist_01_r {parent=lowerarm_r, loc=<-17.8549 0 0>, rot=<0 0 0 -1>}
 deform_elbow_in_base_r {parent=lowerarm_r, loc=<-0.2114 -5.6449 -0.827>, rot=<0 0 0 -1>}
 deform_elbow_in_r {parent=deform_elbow_in_base_r, loc=<0 0 0>, rot=<0 0 0 1>}
 deform_elbow_out_base_r {parent=lowerarm_r, loc=<-0.0001 7.5129 0.0002>, rot=<0 0 0 -1>}
 deform_elbow_out_r {parent=deform_elbow_out_base_r, loc=<0 0 0>, rot=<0 0 0 1>}
 upperarm_twist_01_r {parent=upperarm_r, loc=<-7.9796 -0.0843 0.6195>, rot=<0 0 0 -1>}
 deform_upperarm_bk_base_r {parent=upperarm_twist_01_r, loc=<6.8985 13.3826 1.3257>, rot=<0 0 0 1>}
 deform_upperarm_bk_r {parent=deform_upperarm_bk_base_r, loc=<0 0 0>, rot=<0 0 0 1>}
 deform_upperarm_fr_base_r {parent=upperarm_twist_01_r, loc=<8.3485 -10.7117 2.1456>, rot=<0 0 0 1>}
 deform_upperarm_fr_r {parent=deform_upperarm_fr_base_r, loc=<0 0 0>, rot=<0 0 0 1>}
 deform_upperarm_up_base_r {parent=upperarm_twist_01_r, loc=<11.2015 1.2234 -8.7788>, rot=<0 0 0 1>}
 deform_upperarm_up_r {parent=deform_upperarm_up_base_r, loc=<0 0 0>, rot=<0 0 0 1>}
 upperarm_twist_02_r {parent=upperarm_r, loc=<-19.5059 -0.0843 0.6195>, rot=<0 0 0 -1>}
 deform_bicep_base_r {parent=upperarm_twist_02_r, loc=<0.0002 -8.0494 -0.0003>, rot=<0 0 0 1>}
 deform_bicep_r {parent=deform_bicep_base_r, loc=<0 0 0>, rot=<0 0 0 1>}
 base_whistle_01_m {parent=spine_05, loc=<5.9676 7.9266 0.1334>, rot=<-0.2645 -0.6558 -0.647 -0.2853>}
 dyn_whistle_01_m {parent=base_whistle_01_m, loc=<0 0 0>, rot=<0 0 0 1>}
 dyn_whistle_02_m {parent=dyn_whistle_01_m, loc=<7.311 0 0>, rot=<0.0012 -0.1118 -0.0213 -0.9935>}
 dyn_whistle_03_m {parent=dyn_whistle_02_m, loc=<8.959 0 0>, rot=<0 0 0 1>}
 base_chestHelper_01_l {parent=dyn_whistle_01_m, loc=<3.95 -12.6 0.174>, rot=<0.0068 -0.4255 -0.0144 -0.9048>}
 dyn_chestHelper_01_l {parent=base_chestHelper_01_l, loc=<0 0 0>, rot=<0 0 0 1>}
 base_chestHelper_01_r {parent=dyn_whistle_01_m, loc=<3.139 12.827 0.173>, rot=<0.0068 -0.4255 -0.0144 -0.9048>}
 dyn_chestHelper_01_r {parent=base_chestHelper_01_r, loc=<0 0 0>, rot=<0 0 0 1>}
 spine_05_base_collider {parent=spine_05, loc=<-6.348 16.049 0>, rot=<0 0 0 1>}
 base_necklaceCol_01_l {parent=spine_05, loc=<4.327 9.066 -14.496>, rot=<0.2525 0.6605 0.6605 0.2525>}
 base_necklaceCol_01_r {parent=spine_05, loc=<4.326 9.066 14.426>, rot=<-0.2525 -0.6605 -0.6605 -0.2525>}
 deform_pec_aim_base_l {parent=spine_04, loc=<5.0232 13.6719 -2.6508>, rot=<0.0672 -0.5201 0.3228 -0.7879>}
 deform_pec_aim_l {parent=deform_pec_aim_base_l, loc=<0 0 0>, rot=<0 0 0 1>}
 deform_pec_base_l {parent=deform_pec_aim_l, loc=<16.3279 0 0>, rot=<0 0 0 1>}
 deform_pec_l {parent=deform_pec_base_l, loc=<0 0 0>, rot=<0 0 0 1>}
 deform_pec_aim_base_r {parent=spine_04, loc=<5.023 13.6718 2.6508>, rot=<-0.5201 -0.0672 0.7879 0.3228>}
 deform_pec_aim_r {parent=deform_pec_aim_base_r, loc=<0 0 0>, rot=<0 0 0 1>}
 deform_pec_base_r {parent=deform_pec_aim_r, loc=<-16.3279 0 0>, rot=<0 0 0 -1>}
 deform_pec_r {parent=deform_pec_base_r, loc=<0 0 0.0002>, rot=<0 0 0 -1>}
 deform_lat_l_base {parent=spine_03, loc=<2.5566 -10.3716 -17.6918>, rot=<0 0 0 1>}
 deform_lat_l {parent=deform_lat_l_base, loc=<0 0 0>, rot=<0 0 0 1>}
 deform_lat_r_base {parent=spine_03, loc=<2.5568 -10.3716 17.6918>, rot=<0 0 1 0>}
 deform_lat_r {parent=deform_lat_r_base, loc=<0 0 0>, rot=<0 0 0 -1>}
 thigh_l {parent=pelvis, loc=<-1.8716 -0.0109 -11.9936>, rot=<0.0012 0.0572 0.0155 -0.9982>}
 calf_l {parent=thigh_l, loc=<-41.9178 -0.2246 0.0926>, rot=<0.0006 -0.0161 -0.0391 -0.9991>}
 foot_l {parent=calf_l, loc=<-38.7057 1.2438 0.2452>, rot=<0.0975 -0.0474 0.0317 -0.9936>}
 ball_l {parent=foot_l, loc=<-6.1614 13.1469 0.6807>, rot=<0.0001 -0.0001 -0.7063 -0.7079>}
 calf_twist_02_l {parent=calf_l, loc=<-15.9268 -0.0967 0.0397>, rot=<0.0237 0.0095 -0.0035 -0.9997>}
 calf_twist_01_l {parent=calf_l, loc=<-32.9335 -0.0967 0.0397>, rot=<0.0237 0.0095 -0.0035 -0.9997>}
 deform_knee_in_base_l {parent=calf_l, loc=<1.298 -8.0356 -0.0496>, rot=<0 0 0 -1>}
 deform_knee_in_l {parent=deform_knee_in_base_l, loc=<0 0 0>, rot=<0 0 0 1>}
 deform_knee_out_base_l {parent=calf_l, loc=<0 6.8739 0>, rot=<0 0 0 -1>}
 deform_knee_out_l {parent=deform_knee_out_base_l, loc=<0 0 0>, rot=<0 0 0 1>}
 thigh_twist_01_l {parent=thigh_l, loc=<-21.6303 0 0>, rot=<0.0013 0 0 -1>}
 base_shortFr_01_l {parent=thigh_twist_01_l, loc=<1.2475 10.4659 0.9813>, rot=<-0.001 0.7062 0.708 0.0045>}
 dyn_shortFr_01_l {parent=base_shortFr_01_l, loc=<0 0 0>, rot=<0 0 0 1>}
 base_shortOut_01_l {parent=thigh_twist_01_l, loc=<1.2692 0.0033 -8.4028>, rot=<-0.0024 -1 -0.0013 -0.0039>}
 dyn_shortOut_01_l {parent=base_shortOut_01_l, loc=<0 0 0>, rot=<0 0 0 1>}
 base_shortBck_01_l {parent=thigh_twist_01_l, loc=<1.1458 -10.4586 0.9252>, rot=<-0.0045 -0.708 0.7062 -0.001>}
 dyn_shortBck_01_l {parent=base_shortBck_01_l, loc=<0 0 0>, rot=<0 0 0 1>}
 base_shortInner_01_l {parent=thigh_twist_01_l, loc=<1.1099 -0.051 12.1204>, rot=<-0.0039 -0.0013 1 0.0024>}
 dyn_shortInner_01_l {parent=base_shortInner_01_l, loc=<0 0 0>, rot=<0 0 0 1>}
 deform_groin_base_l {parent=thigh_l, loc=<-2.1908 8.5661 -1.7442>, rot=<0 0 0 -1>}
 deform_groin_l {parent=deform_groin_base_l, loc=<0 0 0>, rot=<0 0 0 1>}
 base_belt_01_l {parent=thigh_l, loc=<-2.7144 11.9911 2.1908>, rot=<0.1509 -0.731 0.1377 -0.6511>}
 thigh_r {parent=pelvis, loc=<-1.8716 -0.0109 11.9936>, rot=<0.0572 -0.0012 0.9982 0.0155>}
 calf_r {parent=thigh_r, loc=<41.9178 0.2246 -0.0926>, rot=<0.0006 -0.0161 -0.0391 -0.9991>}
 foot_r {parent=calf_r, loc=<38.7057 -1.2438 -0.2452>, rot=<0.0975 -0.0474 0.0317 -0.9936>}
 ball_r {parent=foot_r, loc=<6.1614 -13.1469 -0.6807>, rot=<0.0001 -0.0001 -0.7063 -0.7079>}
 calf_twist_02_r {parent=calf_r, loc=<15.9268 0.0967 -0.0397>, rot=<0.0237 0.0095 -0.0035 -0.9997>}
 calf_twist_01_r {parent=calf_r, loc=<32.9335 0.0967 -0.0397>, rot=<0.0237 0.0095 -0.0035 -0.9997>}
 deform_knee_in_base_r {parent=calf_r, loc=<-1.2981 8.0356 0.0496>, rot=<0 0 0 -1>}
 deform_knee_in_r {parent=deform_knee_in_base_r, loc=<0 0 0>, rot=<0 0 0 1>}
 deform_knee_out_base_r {parent=calf_r, loc=<0 -6.8739 0>, rot=<0 0 0 -1>}
 deform_knee_out_r {parent=deform_knee_out_base_r, loc=<0 0 0>, rot=<0 0 0 1>}
 thigh_twist_01_r {parent=thigh_r, loc=<21.6303 0 0>, rot=<0.0013 0 0 -1>}
 base_shortFr_01_r {parent=thigh_twist_01_r, loc=<-1.2475 -10.466 -0.9813>, rot=<-0.708 0.0045 -0.001 -0.7062>}
 dyn_shortFr_01_r {parent=base_shortFr_01_r, loc=<0 0 0>, rot=<0 0 0 1>}
 base_shortOut_01_r {parent=thigh_twist_01_r, loc=<-1.2692 -0.0033 8.4029>, rot=<-0.0013 0.0039 0.0024 -1>}
 dyn_shortOut_01_r {parent=base_shortOut_01_r, loc=<0 0 0>, rot=<0 0 0 1>}
 base_shortBck_01_r {parent=thigh_twist_01_r, loc=<-1.1458 10.4586 -0.9252>, rot=<0.7062 0.001 0.0045 -0.708>}
 dyn_shortBck_01_r {parent=base_shortBck_01_r, loc=<0 0 0>, rot=<0 0 0 1>}
 base_shortInner_01_r {parent=thigh_twist_01_r, loc=<-1.1099 0.051 -12.1205>, rot=<1 -0.0024 0.0039 -0.0013>}
 dyn_shortInner_01_r {parent=base_shortInner_01_r, loc=<0 0 0>, rot=<0 0 0 1>}
 deform_groin_base_r {parent=thigh_r, loc=<2.1908 -8.5661 1.7442>, rot=<0 0 0 -1>}
 deform_groin_r {parent=deform_groin_base_r, loc=<0 0 0>, rot=<0 0 0 -1>}
 deform_glute_base_l {parent=pelvis, loc=<-10.1877 -11.6217 -12.899>, rot=<0.0275 0.0501 -0.4472 -0.8926>}
 deform_glute_l {parent=deform_glute_base_l, loc=<0 0 0>, rot=<0 0 0 1>}
 deform_glute_base_r {parent=pelvis, loc=<-10.1877 -11.6218 12.899>, rot=<0.0501 -0.0275 0.8926 -0.4472>}
 deform_glute_r {parent=deform_glute_base_r, loc=<0 0 0>, rot=<0 0 0 1>}
 base_megaphone_01_r {parent=pelvis, loc=<2.0675 -3.4009 18.168>, rot=<0 0 0.9998 0.0179>}
 dyn_megaphone_01_r {parent=base_megaphone_01_r, loc=<0 0 0>, rot=<0 0 0 1>}
 base_cordA_01_r {parent=pelvis, loc=<-5.7092 14.0306 1.365>, rot=<0.0127 0.707 0.707 0.0127>}
 dyn_cordA_01_r {parent=base_cordA_01_r, loc=<0 0 0>, rot=<0 0 0 1>}
 dyn_cordA_02_r {parent=dyn_cordA_01_r, loc=<4.1807 0 0>, rot=<0 0 0 1>}
 base_cordB_01_r {parent=pelvis, loc=<-5.7092 14.0306 2.2273>, rot=<0.0127 0.707 0.707 0.0127>}
 dyn_cordB_01_r {parent=base_cordB_01_r, loc=<0 0 0>, rot=<0 0 0 1>}
 base_cordA_01_l {parent=pelvis, loc=<-5.983 14.0404 -1.2998>, rot=<0.0127 0.707 0.707 0.0127>}
 dyn_cordA_01_l {parent=base_cordA_01_l, loc=<0 0 0>, rot=<0 0 0 1>}
 base_belt_01_r {parent=pelvis, loc=<-1.5054 11.98 10.098>, rot=<-0.0127 -0.707 0.0127 -0.707>}
]
variables [
 DeformRig_ Curve_deformRig_l_thigh_up_alpha_88_#12 {type=bool, default=True}
 InAir {type=double, default=0}
 LeftMove {type=bool, default=False}
 NewVar {type=double, default=0}
 NewVar_0 {type=double, default=0}
 ParachuteOpen {type=double, default=0}
 RailGrind {type=double, default=0}
 RightMove {type=bool, default=False}
 ShotGunArm {type=double, default=0}
 SkydiveMovement {type=double, default=0}
 crouchWalking {type=double, default=0}
 isDBNO {type=double, default=0}
 largeGunEquipped {type=double, default=0}
 pistolHand {type=double, default=0}
]
