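# Physics asset for the DorsalDance large male player body, setting up secondary motion for its whistle and left-arm rope.
# Both are three-link dyn_ capsule chains with angular limits, hanging from base_ bodies locked to spine_05 and upperarm_l.
# Spine, neck, upper arm and box necklace bodies provide collision shapes, mostly joined by locked constraints.
# The gas can body has free rotation on upperarm_l, while spine_05_base_collider has no constraint.
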PhysicsAsset /Game/Characters/Player/Male/Large/Bodies/M_LRG_DorsalDance/Meshes/M_LRG_DorsalDance_Physics
bodies {
 base_GasCan_01_l [{
   shapeType=Capsule
   length=7.2021
   radius=3.398
   shapeName=upperarm_twist_02_l_capsule
   center=<-3.1426 2.8833 7.6483>
   rotation=<rot 80.1661 -0.1303 -0.1507>
  }]
 base_armRope_01_l [{shapeType=Capsule, length=0.505, radius=0.505, shapeName=base_armRope_01_l_capsule, rotation=<rot 0 -90 0>}]
 base_necklaceCol_01_l [
  {shapeType=Box, extentX=15.688, extentY=13.6623, extentZ=1.0625, shapeName=base_necklaceCol_01_l_box, center=<2.0996 0 -2.3149>}
  {shapeType=Box, extentX=1.2293, extentY=14.2263, extentZ=8.8896, shapeName=base_necklaceCol_01_l_box, center=<-6.1535 0 1.5028>}
 ]
 base_necklaceCol_01_r [
  {shapeType=Box, extentX=15.688, extentY=13.6623, extentZ=1.0625, shapeName=base_necklaceCol_01_r_box, center=<2.0996 0 -1.6367>}
  {shapeType=Box, extentX=1.2293, extentY=14.2263, extentZ=8.8896, shapeName=base_necklaceCol_01_r_box, center=<-5.8818 0 1.5028>}
 ]
 base_whistle_01_m [{shapeType=Capsule, length=0.505, radius=0.505, shapeName=base_whistle_01_m_capsule, rotation=<rot 90 0 0>}]
 dyn_armRope_01_l [{shapeType=Capsule, length=1, radius=0.5, shapeName=dyn_armRope_01_l_capsule, center=<0.8841 0 0>, rotation=<rot 90 0 0>}]
 dyn_armRope_02_l [{shapeType=Capsule, length=1, radius=0.5, shapeName=dyn_armRope_02_l_capsule, center=<0.8841 0 0>, rotation=<rot 90 0 0>}]
 dyn_armRope_03_l [{shapeType=Capsule, length=1.4339, radius=0.5, shapeName=dyn_armRope_03_l_capsule, center=<1.0384 0 0>, rotation=<rot 90 0 0>}]
 dyn_whistle_01_m [
  {shapeType=Capsule, length=5.5572, radius=2.578, shapeName=dyn_whistle_01_m_capsule, center=<3.7038 0 1.4548>, rotation=<rot 90 0 0>}
  {
   shapeType=Capsule
   length=5.5572
   radius=2.578
   shapeName=dyn_whistle_01_m_capsule
   center=<3.7038 -12.0444 1.9151>
   rotation=<rot 87.4706 0 0>
  }
  {shapeType=Capsule, length=5.5572, radius=2.578, shapeName=dyn_whistle_01_m_capsule, center=<3.772 12.0444 1.312>, rotation=<rot 84.8061 0 0>}
 ]
 dyn_whistle_02_m [
  {shapeType=Capsule, length=7.867, radius=2.0409, shapeName=dyn_whistle_02_m_capsule, center=<4.2841 0 0>, rotation=<rot 90 0 0>}
  {shapeType=Capsule, length=6.3281, radius=1.8071, shapeName=dyn_whistle_01_m_capsule, center=<3.7038 -8.2413 0.1136>, rotation=<rot 90 0 0>}
  {shapeType=Capsule, length=6.3281, radius=1.8071, shapeName=dyn_whistle_01_m_capsule, center=<3.7038 9.1505 0.1136>, rotation=<rot 90 0 0>}
 ]
 dyn_whistle_03_m [{shapeType=Capsule, length=5.3686, radius=2.7067, shapeName=dyn_whistle_03_m_capsule, center=<4.224 0 -0.8981>, rotation=<rot 63.5787 0 0>}]
 neck_01 [{shapeType=Capsule, length=10.8957, radius=2.8102, shapeName=neck_01_capsule}]
 spine_04 [{shapeType=Capsule, length=10.5587, radius=17.4569, shapeName=spine_04_capsule, center=<-2.3671 -0.7663 0>}]
 spine_05 [{shapeType=Capsule, length=15.9949, radius=0, shapeName=spine_05_capsule, center=<-8.1204 -0.3183 0>}]
 spine_05_base_collider [{shapeType=Capsule, length=15.9949, radius=15.6967, shapeName=spine_05_capsule, center=<-1.8051 -16.0464 0>}]
 upperarm_l [{shapeType=Capsule, length=15.7702, radius=8.7696, shapeName=upperarm_l_capsule, center=<13.2558 -1.6535 0>, rotation=<rot 90 0 0>}]
}
constraints [
 {bone1Name=dyn_whistle_01_m, bone2Name=base_whistle_01_m, swing1LimitDegrees=0, swing2LimitDegrees=45, twistLimitDegrees=0}
 {bone1Name=dyn_whistle_02_m, bone2Name=dyn_whistle_01_m, swing1LimitDegrees=0, swing2LimitDegrees=45, twistLimitDegrees=0}
 {bone1Name=dyn_whistle_03_m, bone2Name=dyn_whistle_02_m, swing1LimitDegrees=0, swing2LimitDegrees=30, twistLimitDegrees=0}
 {
  bone1Name=base_whistle_01_m
  bone2Name=spine_05
  swing1LimitDegrees=45
  swing1Motion=Locked
  swing2LimitDegrees=45
  swing2Motion=Locked
  twistLimitDegrees=45
  twistMotion=Locked
 }
 {
  bone1Name=neck_01
  bone2Name=spine_05
  swing1LimitDegrees=45
  swing1Motion=Locked
  swing2LimitDegrees=45
  swing2Motion=Locked
  twistLimitDegrees=45
  twistMotion=Locked
 }
 {
  bone1Name=spine_05
  bone2Name=spine_04
  swing1LimitDegrees=45
  swing1Motion=Locked
  swing2LimitDegrees=45
  swing2Motion=Locked
  twistLimitDegrees=45
  twistMotion=Locked
 }
 {
  bone1Name=base_armRope_01_l
  bone2Name=upperarm_l
  swing1LimitDegrees=45
  swing1Motion=Locked
  swing2LimitDegrees=45
  swing2Motion=Locked
  twistLimitDegrees=45
  twistMotion=Locked
 }
 {bone1Name=dyn_armRope_01_l, bone2Name=base_armRope_01_l, swing1LimitDegrees=20, swing2LimitDegrees=20, twistLimitDegrees=0}
 {bone1Name=dyn_armRope_02_l, bone2Name=dyn_armRope_01_l, swing1LimitDegrees=20, swing2LimitDegrees=20, twistLimitDegrees=0}
 {bone1Name=dyn_armRope_03_l, bone2Name=dyn_armRope_02_l, swing1LimitDegrees=20, swing2LimitDegrees=20, twistLimitDegrees=0}
 {
  bone1Name=upperarm_l
  bone2Name=spine_05
  swing1LimitDegrees=45
  swing1Motion=Locked
  swing2LimitDegrees=45
  swing2Motion=Locked
  twistLimitDegrees=45
  twistMotion=Locked
 }
 {
  bone1Name=base_GasCan_01_l
  bone2Name=upperarm_l
  swing1LimitDegrees=45
  swing1Motion=Free
  swing2LimitDegrees=45
  swing2Motion=Free
  twistLimitDegrees=45
  twistMotion=Free
 }
 {
  bone1Name=base_necklaceCol_01_l
  bone2Name=spine_05
  swing1LimitDegrees=45
  swing1Motion=Locked
  swing2LimitDegrees=45
  swing2Motion=Locked
  twistLimitDegrees=45
  twistMotion=Locked
 }
 {
  bone1Name=base_necklaceCol_01_r
  bone2Name=spine_05
  swing1LimitDegrees=45
  swing1Motion=Locked
  swing2LimitDegrees=45
  swing2Motion=Locked
  twistLimitDegrees=45
  twistMotion=Locked
 }
]
