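# Front-end physics asset for the large male DorsalDance player body, covering a dangling left-arm rope and a whistle.
# Capsules on the neck, upper spine, left upper arm and gas can form the body shapes, with neck, spine and arm joints locked.
# Rope and whistle are three-link chains with tight angular limits, hung from base bones on the arm and spine_05.
# The whistle swings mainly on one axis; its base and the gas can have free angular constraints.
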PhysicsAsset /Game/Characters/Player/Male/Large/Bodies/M_LRG_DorsalDance/Meshes/M_LRG_DorsalDance_Physics_FrontEnd
bodies {
 base_GasCan_01_l [{
   shapeType=Capsule
   length=7.2021
   radius=3.398
   shapeName=upperarm_twist_02_l_capsule
   center=<-3.1426 2.8833 7.6483>
   rotation=<rot 80.1661 -0.1303 -0.1507>
  }]
 base_armRope_01_l [{shapeType=Capsule, length=0.505, radius=0.505, shapeName=base_armRope_01_l_capsule, rotation=<rot 0 -90 0>}]
 base_whistle_01_m [{shapeType=Capsule, length=0.505, radius=0.505, shapeName=base_whistle_01_m_capsule, rotation=<rot 0 -90 0>}]
 dyn_armRope_01_l [{shapeType=Capsule, length=1, radius=0.5, shapeName=dyn_armRope_01_l_capsule, center=<0.8841 0 0>, rotation=<rot 90 0 0>}]
 dyn_armRope_02_l [{shapeType=Capsule, length=1, radius=0.5, shapeName=dyn_armRope_02_l_capsule, center=<0.8841 0 0>, rotation=<rot 90 0 0>}]
 dyn_armRope_03_l [{shapeType=Capsule, length=1.4339, radius=0.5, shapeName=dyn_armRope_03_l_capsule, center=<1.0384 0 0>, rotation=<rot 90 0 0>}]
 dyn_whistle_01_m [{shapeType=Capsule, length=21.7191, radius=2, shapeName=dyn_whistle_01_m_capsule, center=<6 0 1>, rotation=<rot 0 0.9165 -89.9456>}]
 dyn_whistle_02_m [{shapeType=Capsule, length=11.7649, radius=1.87, shapeName=dyn_whistle_02_m_capsule, center=<3.9395 0 1>, rotation=<rot 0 0 90>}]
 dyn_whistle_03_m [{
   shapeType=Capsule
   length=5.6243
   radius=2.9
   shapeName=dyn_whistle_03_m_capsule
   center=<4.403 -0.0495 -0.7418>
   rotation=<rot 0 -90.9035 -64.5023>
  }]
 neck_01 [{shapeType=Capsule, length=10.8957, radius=2.8102, shapeName=neck_01_capsule}]
 spine_04 [{shapeType=Capsule, length=10.5587, radius=17.4569, shapeName=spine_04_capsule, center=<-1.667 -0.1034 0>}]
 spine_05 [{shapeType=Capsule, length=15.9949, radius=0, shapeName=spine_05_capsule, center=<-8.1204 -0.3183 0>}]
 spine_05_base_collider [{shapeType=Capsule, length=15.9949, radius=15.6967, shapeName=spine_05_capsule, center=<-1.5691 -16.205 0>}]
 upperarm_l [{shapeType=Capsule, length=15.7702, radius=8.7696, shapeName=upperarm_l_capsule, center=<13.2558 -1.6535 0>, rotation=<rot 90 0 0>}]
}
constraints [
 {
  bone1Name=neck_01
  bone2Name=spine_05
  swing1LimitDegrees=45
  swing1Motion=Locked
  swing2LimitDegrees=45
  swing2Motion=Locked
  twistLimitDegrees=45
  twistMotion=Locked
 }
 {
  bone1Name=spine_05
  bone2Name=spine_04
  swing1LimitDegrees=45
  swing1Motion=Locked
  swing2LimitDegrees=45
  swing2Motion=Locked
  twistLimitDegrees=45
  twistMotion=Locked
 }
 {
  bone1Name=base_armRope_01_l
  bone2Name=upperarm_l
  swing1LimitDegrees=45
  swing1Motion=Locked
  swing2LimitDegrees=45
  swing2Motion=Locked
  twistLimitDegrees=45
  twistMotion=Locked
 }
 {bone1Name=dyn_armRope_01_l, bone2Name=base_armRope_01_l, swing1LimitDegrees=20, swing2LimitDegrees=20, twistLimitDegrees=0}
 {bone1Name=dyn_armRope_02_l, bone2Name=dyn_armRope_01_l, swing1LimitDegrees=20, swing2LimitDegrees=20, twistLimitDegrees=0}
 {bone1Name=dyn_armRope_03_l, bone2Name=dyn_armRope_02_l, swing1LimitDegrees=20, swing2LimitDegrees=20, twistLimitDegrees=0}
 {
  bone1Name=upperarm_l
  bone2Name=spine_05
  swing1LimitDegrees=45
  swing1Motion=Locked
  swing2LimitDegrees=45
  swing2Motion=Locked
  twistLimitDegrees=45
  twistMotion=Locked
 }
 {
  bone1Name=base_GasCan_01_l
  bone2Name=upperarm_l
  swing1LimitDegrees=45
  swing1Motion=Free
  swing2LimitDegrees=45
  swing2Motion=Free
  twistLimitDegrees=45
  twistMotion=Free
 }
 {
  bone1Name=base_whistle_01_m
  bone2Name=spine_05
  swing1LimitDegrees=45
  swing1Motion=Free
  swing2LimitDegrees=45
  swing2Motion=Free
  twistLimitDegrees=45
  twistMotion=Free
 }
 {bone1Name=dyn_whistle_01_m, bone2Name=base_whistle_01_m, swing1LimitDegrees=1, swing2LimitDegrees=25, twistLimitDegrees=1}
 {bone1Name=dyn_whistle_02_m, bone2Name=dyn_whistle_01_m, swing1LimitDegrees=1, swing2LimitDegrees=25, twistLimitDegrees=1}
 {bone1Name=dyn_whistle_03_m, bone2Name=dyn_whistle_02_m, swing1LimitDegrees=1, swing2LimitDegrees=25, twistLimitDegrees=1}
]
