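# Post-process style AnimBP for the DorsalDance large male body, animating its whistle, necklace, megaphone, arm rope and shorts.
# The parent mesh's pose feeds a Control Rig and corrective bone offsets, driven by EventGraph weapon, movement and state alphas.
# In game, RigidBody and AnimDynamics add secondary motion before further offsets.
# The front end instead uses its own physics asset, or whistle constraints when rigid bodies are disabled.
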AnimBlueprint /Game/Characters/Player/Male/Large/Bodies/M_LRG_DorsalDance/Meshes/M_LRG_DorsalDance_AnimBP
target_skeleton=&/M_LRG_DorsalDance_Skeleton
graphs [
 AnimGraph {nodes [
   Local To Component()
   "Transform (Modify) Bone - Bone: dyn_necklace_01_bk"(Alpha=1, Translation="2.585,-0.063,8.486", Rotation="-25.689,4.1484,0", Scale="1,1,1") {
    rotationMode=BMM_Additive
    rotationSpace=BCS_BoneSpace
    scaleMode=BMM_Additive
    scaleSpace=BCS_BoneSpace
    translationMode=BMM_Additive
    translationSpace=BCS_BoneSpace
    alphaBoolBlend {blendOption=Linear}
    boneToModify {boneName=dyn_necklace_01_bk}
    rotation=<rot -25.689 4.1484 0>
    translation=<2.585 -0.063 8.486>
   }
   "Transform (Modify) Bone - Bone: base_necklace_01_bk"!(ComponentPose, Alpha=Get FrontendAlpha() -> (FrontendAlpha:double), Translation="0,0,0", Rotation="0,0,0") -> (Pose) {
    rotationMode=BMM_Additive
    rotationSpace=BCS_BoneSpace
    translationMode=BMM_Additive
    translationSpace=BCS_BoneSpace
    alphaBoolBlend {blendOption=Linear}
    boneToModify {boneName=base_necklace_01_bk}
   }
   "Transform (Modify) Bone - Bone: base_necklace_01_bk"(Alpha=Get Head_Turn_R() -> (Head_Turn_R:double), Translation="0,0,1", Rotation="0,0,0") {
    rotationMode=BMM_Additive
    rotationSpace=BCS_BoneSpace
    translationMode=BMM_Additive
    translationSpace=BCS_BoneSpace
    alphaBoolBlend {blendOption=Linear}
    boneToModify {boneName=base_necklace_01_bk}
    translation {z=1}
   }
   RigidBody!(ComponentPose, Alpha, OverrideWorldGravity:Vector="0,0,0") -> (Pose) {
    bOverrideWorldGravity=true
    overridePhysicsAsset=&/M_LRG_DorsalDance_Physics
    simulationSpace=BaseBoneSpace
    alphaBoolBlend {blendOption=Linear}
    baseBoneRef {boneName=spine_04}
    componentAppliedLinearAccClamp=<1500 1500 1500>
    componentLinearAccScale=<0.6 0.6 0.6>
    componentLinearVelScale=<0.6 0.6 0.6>
    simSpaceSettings {
     maxAngularAcceleration=1000
     maxAngularVelocity=10
     maxLinearAcceleration=700
     maxLinearVelocity=500
     velocityScaleZ=0
     worldAlpha=0.5
     externalLinearDragV=<0.1 0.1>
    }
   }
   n1 = "Constraint - dyn_end_whistle_SC (PPPPPPPP)"!(
    ComponentPose=Local To Component(LocalPose=Use cached pose 'CRCache_PreDyn'())
    Alpha=1
    ConstraintWeights_0=0.3521
    ConstraintWeights_1:float=0.32
    ConstraintWeights_2:float=0.1567
    ConstraintWeights_3:float=0.0829
    ConstraintWeights_4:float=0.0334
    ConstraintWeights_5:float=0.0248
    ConstraintWeights_6:float=0.0221
    ConstraintWeights_7:float=0.0081
   ) -> (Pose) {
    alphaBoolBlend {blendOption=Linear}
    boneToModify {boneName=dyn_end_whistle_SC}
    constraintSetup [
     {targetBone {boneName=spine_04}}
     {targetBone {boneName=spine_03}}
     {targetBone {boneName=spine_02}}
     {targetBone {boneName=spine_05}}
     {targetBone {boneName=spine_01}}
     {targetBone {boneName=deform_pec_r}}
     {targetBone {boneName=deform_pec_l}}
     {targetBone {boneName=clavicle_r}}
    ]
    constraintWeights [1,1,1,1,1,1,1,1]
   }
   n2 = "Constraint - dyn_whistle_03_m (P)"(ComponentPose=@n1, Alpha=0.4, ConstraintWeights_0=1) {
    alphaBoolBlend {blendOption=Linear}
    boneToModify {boneName=dyn_whistle_03_m}
    constraintSetup [{targetBone {boneName=dyn_end_whistle_SC}}]
    constraintWeights [1]
   }
   n3 = "Constraint - dyn_whistle_02_m (P)"(ComponentPose=@n2, Alpha=0.5, ConstraintWeights_0=1) {
    alphaBoolBlend {blendOption=Linear}
    boneToModify {boneName=dyn_whistle_02_m}
    constraintSetup [{targetBone {boneName=dyn_end_whistle_SC}}]
    constraintWeights [1]
   }
   n4 = "Constraint - dyn_whistle_02_m (U)"(ComponentPose=@n3, Alpha=1, ConstraintWeights_0=1) {
    alphaBoolBlend {blendOption=Linear}
    boneToModify {boneName=dyn_whistle_02_m}
    constraintSetup [{transformType=LookAt, targetBone {boneName=base_whistle_01_m}}]
    constraintWeights [1]
   }
   n5 = "Transform (Modify) Bone - Bone: dyn_whistle_03_m"(ComponentPose=@n4, Alpha=1, Translation="0,0,0", Rotation="5,0,0", Scale="1,1,1") {
    rotationMode=BMM_Additive
    rotationSpace=BCS_BoneSpace
    alphaBoolBlend {blendOption=Linear}
    boneToModify {boneName=dyn_whistle_03_m}
    rotation {pitch=5}
   }
   n6 = "Transform (Modify) Bone - Bone: base_whistle_01_m"!(ComponentPose=@n5, bAlphaBoolEnabled:bool=True, Translation="0,0,1", Rotation="0,0,0", Scale="1,1,1") -> (Pose) {
    alphaInputType=Bool
    rotationMode=BMM_Additive
    rotationSpace=BCS_BoneSpace
    translationMode=BMM_Additive
    translationSpace=BCS_BoneSpace
    alphaBoolBlend {blendInTime=0.1, blendOption=Linear, blendOutTime=0.1}
    boneToModify {boneName=base_whistle_01_m}
    translation {z=1}
   }
   n7 = Component To Local(ComponentPose=@n6)
   n8 = RigidBody(ComponentPose=Local To Component(LocalPose=Use cached pose 'CRCache_PreDyn'()), OverrideWorldGravity="0,0,-1280") {
    bOverrideWorldGravity=true
    overridePhysicsAsset=&/M_LRG_DorsalDance_Physics_FrontEnd
    simulationSpace=BaseBoneSpace
    alphaBoolBlend {blendOption=Linear}
    baseBoneRef {boneName=spine_05}
    componentAppliedLinearAccClamp=<1500 1500 1500>
    componentLinearAccScale=<0.6 0.6 0.6>
    componentLinearVelScale=<0.6 0.6 0.6>
    simSpaceSettings {
     maxAngularAcceleration=1000
     maxAngularVelocity=10
     maxLinearAcceleration=700
     maxLinearVelocity=500
     velocityScaleZ=0
     worldAlpha=0.5
     externalLinearDragV=<0.1 0.1>
    }
   }
   n9 = "Transform (Modify) Bone - Bone: base_whistle_01_m"!(ComponentPose=@n8, Alpha:float=1, Translation="0,0,0", Rotation=Get spine rotation() -> (spine rotation:Rotator), Scale="1,1,1") -> (Pose) {
    rotationMode=BMM_Additive
    rotationSpace=BCS_BoneSpace
    translationMode=BMM_Additive
    translationSpace=BCS_BoneSpace
    alphaBoolBlend {blendOption=Linear}
    boneToModify {boneName=base_whistle_01_m}
   }
   n10 = "Transform (Modify) Bone - Bone: base_belt_01_l"!(ComponentPose=@n9, bAlphaBoolEnabled:bool=True, Translation="7.978,0.404,4.330", Rotation="23.238,-0.000,0.000") -> (Pose) {
    alphaInputType=Bool
    rotationMode=BMM_Additive
    rotationSpace=BCS_BoneSpace
    translationMode=BMM_Additive
    translationSpace=BCS_BoneSpace
    alphaBoolBlend {blendOption=Linear}
    boneToModify {boneName=base_belt_01_l}
    rotation=<rot 23.238 0 0>
    translation=<7.978 0.404 4.33>
   }
   n11 = Get FrontendAlpha()
   n12 = "Transform (Modify) Bone - Bone: dyn_whistle_03_m"!(ComponentPose=@n10, Alpha=@n11, Translation="0,0,-0.9", Rotation="0,0,0") -> (Pose) {
    rotationMode=BMM_Additive
    rotationSpace=BCS_BoneSpace
    translationMode=BMM_Additive
    translationSpace=BCS_BoneSpace
    alphaBoolBlend {blendOption=Linear}
    boneToModify {boneName=dyn_whistle_03_m}
    rotation=<rot 0.0001 0 0>
    translation {z=-0.9}
   }
   n13 = "Transform (Modify) Bone - Bone: dyn_whistle_02_m"!(ComponentPose=@n12, Alpha=@n11, Translation="0,0,-3.5", Rotation="20.000,-0.000,-0.000") -> (Pose) {
    rotationMode=BMM_Additive
    rotationSpace=BCS_BoneSpace
    translationMode=BMM_Additive
    translationSpace=BCS_BoneSpace
    alphaBoolBlend {blendOption=Linear}
    boneToModify {boneName=dyn_whistle_02_m}
    rotation=<rot 20.0001 0 0>
    translation {z=-3.5}
   }
   n14 = Component To Local(ComponentPose=@n13)
   n15 = Get bRigidBodyEnabled()
   n16 = "Transform (Modify) Bone - Bone: base_shortInner_01_l"!(
    ComponentPose=Local To Component(LocalPose=Use cached pose 'CRCache_postDyn'())
    Alpha=1
    Translation="-0.000,2.070,-1.489"
    Rotation="0,0,0"
   ) -> (Pose) {
    rotationMode=BMM_Additive
    rotationSpace=BCS_BoneSpace
    translationMode=BMM_Additive
    translationSpace=BCS_BoneSpace
    alphaBoolBlend {blendOption=Linear}
    alphaScaleBiasClamp {bInterpResult=true}
    boneToModify {boneName=base_shortInner_01_l}
    translation=<0 2.07 -1.4892>
   }
   n17 = "Transform (Modify) Bone - Bone: base_shortInner_01_r"!(ComponentPose=@n16, Alpha=1, Translation="-0.000,-4.147,-1.754", Rotation="0,0,0") -> (Pose) {
    rotationMode=BMM_Additive
    rotationSpace=BCS_BoneSpace
    translationMode=BMM_Additive
    translationSpace=BCS_BoneSpace
    alphaBoolBlend {blendOption=Linear}
    alphaScaleBiasClamp {bInterpResult=true}
    boneToModify {boneName=base_shortInner_01_r}
    translation {y=-4.147, z=-1.754}
   }
   n18 = "Transform (Modify) Bone - Bone: dyn_megaphone_01_r"!(ComponentPose=@n17, Alpha=1, Rotation="30,30,0.832") -> (Pose) {
    rotationMode=BMM_Additive
    rotationSpace=BCS_BoneSpace
    translationMode=BMM_Additive
    translationSpace=BCS_BoneSpace
    alphaBoolBlend {blendOption=Linear}
    alphaScaleBiasClamp {bInterpResult=true, interpSpeedDecreasing=5, interpSpeedIncreasing=5}
    boneToModify {boneName=dyn_megaphone_01_r}
    rotation=<rot 30 30 0.832>
   }
   n19 = "Transform (Modify) Bone - Bone: dyn_megaphone_01_r"(ComponentPose=@n18, Alpha=1, Rotation="0.016,0.058,25.571") {
    rotationMode=BMM_Additive
    rotationSpace=BCS_BoneSpace
    translationMode=BMM_Additive
    translationSpace=BCS_BoneSpace
    alphaBoolBlend {blendOption=Linear}
    alphaScaleBiasClamp {bInterpResult=true, interpSpeedDecreasing=5, interpSpeedIncreasing=5}
    boneToModify {boneName=dyn_megaphone_01_r}
    rotation=<rot 0.016 0.058 25.571>
   }
   n20 = "Transform (Modify) Bone - Bone: dyn_megaphone_01_r"!(ComponentPose=@n19, Alpha=1, Translation:Vector="-0.570,-2.747,1.529", Rotation="-10.775,-123.540,0.576") -> (Pose) {
    rotationMode=BMM_Additive
    rotationSpace=BCS_BoneSpace
    translationMode=BMM_Additive
    translationSpace=BCS_BoneSpace
    alphaBoolBlend {blendOption=Linear}
    alphaScaleBiasClamp {bInterpResult=true, interpSpeedDecreasing=5, interpSpeedIncreasing=5}
    boneToModify {boneName=dyn_megaphone_01_r}
    rotation=<rot -10.775 -123.54 0.576>
    translation=<-0.57 -2.747 1.529>
   }
   n21 = "Transform (Modify) Bone - Bone: base_belt_01_l"(ComponentPose=@n20, bAlphaBoolEnabled=True, Translation="7.978,0.404,4.330", Rotation="23.238,-0.000,0.000") {
    alphaInputType=Bool
    rotationMode=BMM_Additive
    rotationSpace=BCS_BoneSpace
    translationMode=BMM_Additive
    translationSpace=BCS_BoneSpace
    alphaBoolBlend {blendOption=Linear}
    boneToModify {boneName=base_belt_01_l}
    rotation=<rot 23.238 0 0>
    translation=<7.978 0.404 4.33>
   }
   n22 = "Transform (Modify) Bone - Bone: base_chestHelper_01_l"!(ComponentPose=@n21, Alpha=1, Translation="-0.115,3.599,3.686", Rotation="-66.832,0.000,-0.000") -> (Pose) {
    rotationMode=BMM_Additive
    rotationSpace=BCS_BoneSpace
    translationMode=BMM_Additive
    translationSpace=BCS_BoneSpace
    alphaBoolBlend {blendOption=Linear}
    alphaScaleBiasClamp {bInterpResult=true, interpSpeedDecreasing=5}
    boneToModify {boneName=base_chestHelper_01_l}
    rotation=<rot -66.832 0.0001 -0.0001>
    translation=<-0.1151 3.599 3.6863>
   }
   n23 = "Transform (Modify) Bone - Bone: base_necklace_01_l"!(ComponentPose=@n22, Alpha=1, Translation="10.814,-9.583,6.569", Rotation="0.000,0.000,-54.660") -> (Pose) {
    rotationMode=BMM_Additive
    rotationSpace=BCS_BoneSpace
    translationMode=BMM_Additive
    translationSpace=BCS_BoneSpace
    alphaBoolBlend {blendOption=Linear}
    alphaScaleBiasClamp {bInterpResult=true, interpSpeedDecreasing=2}
    boneToModify {boneName=base_necklace_01_l}
    rotation=<rot 0 0 -54.66>
    translation=<10.814 -9.5833 6.5686>
   }
   n24 = "Transform (Modify) Bone - Bone: base_necklace_01_r"!(ComponentPose=@n23, Alpha=1, Translation="-5,0,0", Rotation="0,0,0") -> (Pose) {
    rotationMode=BMM_Additive
    rotationSpace=BCS_BoneSpace
    translationMode=BMM_Additive
    translationSpace=BCS_BoneSpace
    alphaBoolBlend {blendOption=Linear}
    alphaScaleBiasClamp {bInterpResult=true, interpSpeedDecreasing=2}
    boneToModify {boneName=base_necklace_01_r}
    translation {x=-5}
   }
   n25 = "Transform (Modify) Bone - Bone: base_necklace_01_bk"(ComponentPose=@n24, Alpha=1, Translation="0,0,2", Rotation="0,0,0") {
    rotationMode=BMM_Additive
    rotationSpace=BCS_BoneSpace
    translationMode=BMM_Additive
    translationSpace=BCS_BoneSpace
    alphaBoolBlend {blendOption=Linear}
    alphaScaleBiasClamp {bInterpResult=true, interpSpeedDecreasing=2}
    boneToModify {boneName=base_necklace_01_bk}
    translation {z=2}
   }
   n26 = "Transform (Modify) Bone - Bone: base_necklace_01_r"!(ComponentPose=@n25, Alpha=1, Translation="-2.9117,0.2003,0", Rotation="0,17.1819,0", Scale:Vector="1,1,1") -> (Pose) {
    rotationMode=BMM_Additive
    rotationSpace=BCS_BoneSpace
    translationMode=BMM_Additive
    translationSpace=BCS_BoneSpace
    alphaBoolBlend {blendOption=Linear}
    boneToModify {boneName=base_necklace_01_r}
    rotation {yaw=17.1819}
    translation=<-2.9117 0.2003>
   }
   n27 = "Transform (Modify) Bone - Bone: base_armRope_01_l"(ComponentPose=@n26, Alpha=1, Translation="0,0,0", Rotation="23.168,-16.747,10.422", Scale="1,1,1") {
    rotationMode=BMM_Additive
    rotationSpace=BCS_BoneSpace
    translationMode=BMM_Additive
    translationSpace=BCS_BoneSpace
    alphaBoolBlend {blendOption=Linear}
    boneToModify {boneName=base_armRope_01_l}
    rotation=<rot 23.168 -16.747 10.422>
   }
   n28 = "Transform (Modify) Bone - Bone: base_necklace_01_r"(ComponentPose=@n27, Alpha=1, Translation="-0.000,-1.991,2.676", Rotation="0,0,20", Scale="1.000,1.000,0.771") {
    rotationMode=BMM_Additive
    rotationSpace=BCS_BoneSpace
    scaleMode=BMM_Additive
    scaleSpace=BCS_BoneSpace
    translationMode=BMM_Additive
    translationSpace=BCS_BoneSpace
    alphaBoolBlend {blendOption=Linear}
    boneToModify {boneName=base_necklace_01_r}
    rotation {roll=20}
    scale {z=0.771}
    translation {y=-1.991, z=2.676}
   }
   n29 = "Transform (Modify) Bone - Bone: base_necklace_01_r"(ComponentPose=@n28, Alpha=1, Translation="3.667,0.016,0.241", Rotation="-25.689,4.1484,0", Scale="1,1,1") {
    rotationMode=BMM_Additive
    rotationSpace=BCS_BoneSpace
    scaleMode=BMM_Additive
    scaleSpace=BCS_BoneSpace
    translationMode=BMM_Additive
    translationSpace=BCS_BoneSpace
    alphaBoolBlend {blendOption=Linear}
    boneToModify {boneName=base_necklace_01_r}
    rotation=<rot -25.689 4.1484 0>
    translation=<3.667 0.016 0.241>
   }
   Output Pose(Result=Blend Poses by bool(
     BlendPose_0=Blend Poses by bool(BlendPose_0=@n14, BlendPose_1=@n7, bActiveValue=@n15) {blendType=Linear}
     BlendPose_1=Component To Local(ComponentPose=@n29)
     bActiveValue=Get bIsInFrontEnd()
    ) {blendType=Linear}) {layerGroup=None}
   n30 = Control Rig!(
    Source=Copy Pose From Mesh() {bUseAttachedParent=true}
    Alpha
    crouchWalking:double=Get CrouchIdleAlpha() -> (CrouchIdleAlpha:double)
    isDBNO:double=Get DBNOAlpha() -> (DBNOAlpha:double)
    largeGunEquipped:double=Get large_gun_out() -> (large_gun_out:double)
    ParachuteOpen:double=Get ParachuteAlpha() -> (ParachuteAlpha:double)
    RailGrind:double=Get RailAlpha() -> (RailAlpha:double)
    SkydiveMovement:double=Get SkydiveFloatingAlpha() -> (SkydiveFloatingAlpha:double)
    pistolHand:double=Get AlphaPistol() -> (AlphaPistol:float)
    ShotGunArm:double=Get shotgunEquipped() -> (shotgunEquipped:double)
    InAir:double=Get InAir() -> (InAir:double)
    LeftMove:bool=Get LeftBool() -> (LeftBool:bool)
    RightMove:bool=Get RightBool() -> (RightBool:bool)
   ) -> (Pose) {
    alphaBoolBlend {blendOption=Linear}
    alphaScaleBiasClamp {bInterpResult=true}
    controlRigAssetReference {blueprintRigClass=&/M_LRG_DorsalDance_CtrlRig.M_LRG_DorsalDance_CtrlRig_C}
   }
   Save cached pose 'CRCache_input'(Pose=@n30) {cacheName=CRCache_input}
   n31 = RigidBody(ComponentPose=Local To Component(LocalPose=Use cached pose 'CRCache_PreDyn'()), OverrideWorldGravity="0,0,0") {
    bOverrideWorldGravity=true
    overridePhysicsAsset=&M_LRG_DorsalDance_Physics
    simulationSpace=BaseBoneSpace
    alphaBoolBlend {blendOption=Linear}
    baseBoneRef {boneName=spine_04}
    componentAppliedLinearAccClamp=<1500 1500 1500>
    componentLinearAccScale=<0.6 0.6 0.6>
    componentLinearVelScale=<0.6 0.6 0.6>
    simSpaceSettings {
     maxAngularAcceleration=1000
     maxAngularVelocity=10
     maxLinearAcceleration=700
     maxLinearVelocity=500
     velocityScaleZ=0
     worldAlpha=0.5
     externalLinearDragV=<0.1 0.1>
    }
   }
   n32 = Anim Dynamics - Bone: dyn_armRope_01_l!(ComponentPose=@n31, Alpha=1, GravityOverride:Vector="0,0,-100") -> (Pose) {
    angularDampingOverride=0.99
    angularSpringConstant=8000
    bAngularSpring=true
    bOverrideAngularDamping=true
    bUseGravityOverride=true
    bUsePlanarLimit=false
    gravityScale=0.35
    alphaBoolBlend {blendOption=Linear}
    boundBone {boneName=dyn_armRope_01_l}
    componentLinearAccScale=<1 1 1>
    componentLinearVelScale=<1 1 1>
    gravityOverride {z=-980}
    physicsBodyDefinitions [{
      boundBone {boneName=dyn_armRope_01_l}
      boxExtents=<10 20 20>
      constraintSetup {angularLimitsMax=<0 1 40>, angularLimitsMin=<0 -100 -70>, angularTarget=<1 0 0>}
      localJointOffset=<5 0 0>
     }]
   }
   n33 = Anim Dynamics - Bone: dyn_armRope_03_l!(ComponentPose=@n32, Alpha=1, GravityOverride:Vector="0,0,-980") -> (Pose) {
    angularDampingOverride=0.8
    angularSpringConstant=1000
    bAngularSpring=true
    bOverrideAngularDamping=true
    bUseGravityOverride=true
    bUsePlanarLimit=false
    gravityScale=0.35
    alphaBoolBlend {blendOption=Linear}
    boundBone {boneName=dyn_armRope_03_l}
    componentLinearAccScale=<1 1 1>
    componentLinearVelScale=<1 1 1>
    gravityOverride {z=-980}
    physicsBodyDefinitions [{
      boundBone {boneName=dyn_armRope_03_l}
      boxExtents=<10 20 20>
      constraintSetup {angularLimitsMax=<0 1 25>, angularLimitsMin=<0 -60 -25>, angularTarget=<1 0 0>}
      localJointOffset=<5 0 0>
     }]
   }
   n34 = Anim Dynamics - Bone: dyn_megaphone_01_r!(ComponentPose=@n33, Alpha=1, GravityOverride:Vector="0,0,-980") -> (Pose) {
    angularDampingOverride=1
    angularSpringConstant=20000
    bAngularSpring=true
    bOverrideAngularDamping=true
    bUseGravityOverride=true
    bUsePlanarLimit=false
    gravityScale=0.35
    simulationSpace=RootRelative
    alphaBoolBlend {blendOption=Linear}
    boundBone {boneName=dyn_megaphone_01_r}
    componentAppliedLinearAccClamp=<50000 50000 50000>
    componentLinearAccScale=<0.4 0.4 0.4>
    componentLinearVelScale=<0.4 0.4 0.4>
    gravityOverride {z=-980}
    physicsBodyDefinitions [{
      boundBone {boneName=dyn_megaphone_01_r}
      boxExtents=<20 30 30>
      constraintSetup {angularLimitsMax=<0 5 10>, angularLimitsMin=<0 -70 -5>, angularTarget=<1 0 0>}
      localJointOffset=<4 0 0>
     }]
   }
   n35 = Anim Dynamics - Bone: dyn_shortFr_01_l!(ComponentPose=@n34, Alpha=1, GravityOverride:Vector="0,0,-980") -> (Pose) {
    angularDampingOverride=1
    angularSpringConstant=20000
    bAngularSpring=true
    bOverrideAngularDamping=true
    bUseGravityOverride=true
    gravityScale=0.35
    simulationSpace=BoneRelative
    alphaBoolBlend {blendOption=Linear}
    boundBone {boneName=dyn_shortFr_01_l}
    componentLinearAccScale=<1 1 1>
    componentLinearVelScale=<1 1 1>
    gravityOverride {z=-980}
    physicsBodyDefinitions [{
      boundBone {boneName=dyn_shortFr_01_l}
      boxExtents=<20 30 30>
      constraintSetup {angularLimitsMax=<0 3 15>, angularLimitsMin=<0 -30 -15>, angularTarget=<1 0 0>}
      localJointOffset=<5 0 0>
     }]
    relativeSpaceBone {boneName=pelvis}
   }
   n36 = Anim Dynamics - Bone: dyn_shortOut_01_l!(ComponentPose=@n35, Alpha=1, GravityOverride:Vector="0,0,-980") -> (Pose) {
    angularDampingOverride=1
    angularSpringConstant=20000
    bAngularSpring=true
    bOverrideAngularDamping=true
    bUseGravityOverride=true
    gravityScale=0.35
    simulationSpace=BoneRelative
    alphaBoolBlend {blendOption=Linear}
    boundBone {boneName=dyn_shortOut_01_l}
    componentLinearAccScale=<1 1 1>
    componentLinearVelScale=<1 1 1>
    gravityOverride {z=-980}
    physicsBodyDefinitions [{
      boundBone {boneName=dyn_shortOut_01_l}
      boxExtents=<20 30 30>
      constraintSetup {angularLimitsMax=<0 3 15>, angularLimitsMin=<0 -30 -15>, angularTarget=<1 0 0>}
      localJointOffset=<5 0 0>
     }]
    relativeSpaceBone {boneName=pelvis}
   }
   n37 = Anim Dynamics - Bone: dyn_shortBck_01_l!(ComponentPose=@n36, Alpha=1, GravityOverride:Vector="0,0,-980") -> (Pose) {
    angularDampingOverride=1
    angularSpringConstant=20000
    bAngularSpring=true
    bOverrideAngularDamping=true
    bUseGravityOverride=true
    gravityScale=0.35
    simulationSpace=BoneRelative
    alphaBoolBlend {blendOption=Linear}
    boundBone {boneName=dyn_shortBck_01_l}
    componentLinearAccScale=<1 1 1>
    componentLinearVelScale=<1 1 1>
    gravityOverride {z=-980}
    physicsBodyDefinitions [{
      boundBone {boneName=dyn_shortBck_01_l}
      boxExtents=<20 30 30>
      constraintSetup {angularLimitsMax=<0 3 5>, angularLimitsMin=<0 -30 -5>, angularTarget=<1 0 0>}
      localJointOffset=<5 0 0>
     }]
    relativeSpaceBone {boneName=pelvis}
   }
   n38 = Anim Dynamics - Bone: dyn_shortInner_01_l!(ComponentPose=@n37, Alpha=1, GravityOverride:Vector="0,0,-980") -> (Pose) {
    angularDampingOverride=1
    angularSpringConstant=20000
    bAngularSpring=true
    bOverrideAngularDamping=true
    bUseGravityOverride=true
    gravityScale=0.35
    simulationSpace=BoneRelative
    alphaBoolBlend {blendOption=Linear}
    boundBone {boneName=dyn_shortInner_01_l}
    componentLinearAccScale=<1 1 1>
    componentLinearVelScale=<1 1 1>
    gravityOverride {z=-980}
    physicsBodyDefinitions [{
      boundBone {boneName=dyn_shortInner_01_l}
      boxExtents=<20 30 30>
      constraintSetup {angularLimitsMax=<0 -5 15>, angularLimitsMin=<0 -10 -15>, angularTarget=<1 0 0>}
      localJointOffset=<1 0 0>
     }]
    relativeSpaceBone {boneName=pelvis}
   }
   n39 = Anim Dynamics - Bone: dyn_shortFr_01_r!(ComponentPose=@n38, Alpha=1, GravityOverride:Vector="0,0,-980") -> (Pose) {
    angularDampingOverride=1
    angularSpringConstant=20000
    bAngularSpring=true
    bOverrideAngularDamping=true
    bUseGravityOverride=true
    gravityScale=0.35
    simulationSpace=BoneRelative
    alphaBoolBlend {blendOption=Linear}
    boundBone {boneName=dyn_shortFr_01_r}
    componentLinearAccScale=<1 1 1>
    componentLinearVelScale=<1 1 1>
    gravityOverride {z=-980}
    physicsBodyDefinitions [{
      boundBone {boneName=dyn_shortFr_01_r}
      boxExtents=<20 30 30>
      constraintSetup {angularLimitsMax=<0 3 15>, angularLimitsMin=<0 -30 -15>, angularTarget=<1 0 0>}
      localJointOffset=<5 0 0>
     }]
    relativeSpaceBone {boneName=pelvis}
   }
   n40 = Anim Dynamics - Bone: dyn_shortOut_01_r!(ComponentPose=@n39, Alpha=1, GravityOverride:Vector="0,0,-980") -> (Pose) {
    angularDampingOverride=1
    angularSpringConstant=20000
    bAngularSpring=true
    bOverrideAngularDamping=true
    bUseGravityOverride=true
    gravityScale=0.35
    simulationSpace=BoneRelative
    alphaBoolBlend {blendOption=Linear}
    boundBone {boneName=dyn_shortOut_01_r}
    componentLinearAccScale=<1 1 1>
    componentLinearVelScale=<1 1 1>
    gravityOverride {z=-980}
    physicsBodyDefinitions [{
      boundBone {boneName=dyn_shortOut_01_r}
      boxExtents=<20 30 30>
      constraintSetup {angularLimitsMax=<0 3 15>, angularLimitsMin=<0 -30 -15>, angularTarget=<1 0 0>}
      localJointOffset=<5 0 0>
     }]
    relativeSpaceBone {boneName=pelvis}
   }
   n41 = Anim Dynamics - Bone: dyn_shortBck_01_r!(ComponentPose=@n40, Alpha=1, GravityOverride:Vector="0,0,-980") -> (Pose) {
    angularDampingOverride=1
    angularSpringConstant=20000
    bAngularSpring=true
    bOverrideAngularDamping=true
    bUseGravityOverride=true
    gravityScale=0.35
    simulationSpace=BoneRelative
    alphaBoolBlend {blendOption=Linear}
    boundBone {boneName=dyn_shortBck_01_r}
    componentLinearAccScale=<1 1 1>
    componentLinearVelScale=<1 1 1>
    gravityOverride {z=-980}
    physicsBodyDefinitions [{
      boundBone {boneName=dyn_shortBck_01_r}
      boxExtents=<20 30 30>
      constraintSetup {angularLimitsMax=<0 3 5>, angularLimitsMin=<0 -30 -5>, angularTarget=<1 0 0>}
      localJointOffset=<5 0 0>
     }]
    relativeSpaceBone {boneName=pelvis}
   }
   n42 = Anim Dynamics - Bone: dyn_shortInner_01_r!(ComponentPose=@n41, Alpha=1, GravityOverride:Vector="0,0,-980") -> (Pose) {
    angularDampingOverride=1
    angularSpringConstant=20000
    bAngularSpring=true
    bOverrideAngularDamping=true
    bUseGravityOverride=true
    gravityScale=0.35
    simulationSpace=BoneRelative
    alphaBoolBlend {blendOption=Linear}
    boundBone {boneName=dyn_shortInner_01_r}
    componentLinearAccScale=<1 1 1>
    componentLinearVelScale=<1 1 1>
    gravityOverride {z=-980}
    physicsBodyDefinitions [{
      boundBone {boneName=dyn_shortInner_01_r}
      boxExtents=<20 30 30>
      constraintSetup {angularLimitsMax=<0 -5 15>, angularLimitsMin=<0 -10 -15>, angularTarget=<1 0 0>}
      localJointOffset=<1 0 0>
     }]
    relativeSpaceBone {boneName=pelvis}
   }
   Save cached pose 'CRCache_postDyn'(Pose=Component To Local(ComponentPose=@n42)) {cacheName=CRCache_postDyn}
   n43 = "Transform (Modify) Bone - Bone: head"!(ComponentPose=Local To Component(LocalPose=Use cached pose 'CRCache_input'()), Alpha=1, Rotation="50,50,50") -> (Pose) {
    rotationMode=BMM_Additive
    rotationSpace=BCS_BoneSpace
    alphaBoolBlend {blendOption=Linear}
    boneToModify {boneName=head}
    rotation=<rot 50 50 50>
   }
   n44 = "Transform (Modify) Bone - Bone: base_chestHelper_01_r"!(ComponentPose=@n43, Alpha=1, Translation="0,2,8") -> (Pose) {
    rotationMode=BMM_Additive
    rotationSpace=BCS_BoneSpace
    translationMode=BMM_Additive
    translationSpace=BCS_BoneSpace
    alphaBoolBlend {blendOption=Linear}
    alphaScaleBiasClamp {bInterpResult=true}
    boneToModify {boneName=base_chestHelper_01_r}
    translation {y=2, z=8}
   }
   n45 = "Transform (Modify) Bone - Bone: base_chestHelper_01_l"!(ComponentPose=@n44, Alpha=1, Translation="0,0,8") -> (Pose) {
    rotationMode=BMM_Additive
    rotationSpace=BCS_BoneSpace
    translationMode=BMM_Additive
    translationSpace=BCS_BoneSpace
    alphaBoolBlend {blendOption=Linear}
    alphaScaleBiasClamp {bInterpResult=true}
    boneToModify {boneName=base_chestHelper_01_l}
    translation {z=8}
   }
   n46 = "Transform (Modify) Bone - Bone: base_necklace_01_l"(ComponentPose=@n45, Alpha=1, Translation="9.676,0.952,12", Rotation="0,-30,0") {
    rotationMode=BMM_Additive
    rotationSpace=BCS_BoneSpace
    translationMode=BMM_Additive
    translationSpace=BCS_BoneSpace
    alphaBoolBlend {blendOption=Linear}
    alphaScaleBiasClamp {bInterpResult=true, interpSpeedDecreasing=5, interpSpeedIncreasing=5}
    boneToModify {boneName=base_necklace_01_l}
    rotation {yaw=-30}
    translation=<9.676 0.952 12>
   }
   n47 = "Transform (Modify) Bone - Bone: base_belt_01_l"!(ComponentPose=@n46, Alpha:float=1, Translation="0.390,10.123,-0.008", Rotation="-1.972,-2.431,-0.000") -> (Pose) {
    rotationMode=BMM_Additive
    rotationSpace=BCS_BoneSpace
    translationMode=BMM_Additive
    translationSpace=BCS_BoneSpace
    alphaBoolBlend {blendOption=Linear}
    boneToModify {boneName=base_belt_01_l}
    rotation {pitch=-1.972, yaw=-2.431}
    translation=<0.39 10.123 -0.008>
   }
   n48 = "Transform (Modify) Bone - Bone: base_belt_01_r"!(ComponentPose=@n47, Alpha=1, Translation="2.513,3.573,1.578", Rotation="-0.000,-2.431,-26.093") -> (Pose) {
    rotationMode=BMM_Additive
    rotationSpace=BCS_BoneSpace
    translationMode=BMM_Additive
    translationSpace=BCS_BoneSpace
    alphaBoolBlend {blendOption=Linear}
    boneToModify {boneName=base_belt_01_r}
    rotation=<rot 0 -2.431 -26.093>
    translation=<2.513 3.573 1.578>
   }
   n49 = "Transform (Modify) Bone - Bone: base_necklace_01_l"(ComponentPose=@n48, Alpha=1, Translation="0,0,5", Rotation="0,0,0") {
    translationMode=BMM_Additive
    translationSpace=BCS_BoneSpace
    alphaBoolBlend {blendOption=Linear}
    boneToModify {boneName=base_necklace_01_l}
    translation {z=5}
   }
   n50 = "Transform (Modify) Bone - Bone: base_necklace_01_bk"(ComponentPose=@n49, Alpha=1, Translation="0,0,1", Rotation="0,0,0") {
    rotationMode=BMM_Additive
    rotationSpace=BCS_BoneSpace
    translationMode=BMM_Additive
    translationSpace=BCS_BoneSpace
    alphaBoolBlend {blendOption=Linear}
    boneToModify {boneName=base_necklace_01_bk}
    translation {z=1}
   }
   n51 = "Transform (Modify) Bone - Bone: base_necklace_01_r"!(ComponentPose=@n50, Alpha=Get FrontendAlpha(), Translation="-2,0,-1", Rotation="0,0,0") -> (Pose) {
    rotationMode=BMM_Additive
    rotationSpace=BCS_BoneSpace
    translationMode=BMM_Additive
    translationSpace=BCS_BoneSpace
    alphaBoolBlend {blendOption=Linear}
    boneToModify {boneName=base_necklace_01_r}
    translation {x=-2, z=-1}
   }
   Save cached pose 'CRCache_PreDyn'(Pose=Component To Local(ComponentPose=@n51)) {cacheName=CRCache_PreDyn}
   Get HeadRotL() -> (HeadRotL:double)
  ]}
 EventGraph {nodes [
   Delay(Duration=0)
   > ResetDynamics(InTeleportType=TeleportPhysics)
   > ResetDynamics(InTeleportType=ResetPhysics)
   n1 = Event BlueprintUpdateAnimation()
   > n2 = Sequence!(execute) -> (then_0, then_1, then_2:Exec, then_3:Exec, then_4:Exec, then_5:Exec, then_6:Exec)
   n3 = Set Skel Mesh(execute=@n2.then_4, Skel Mesh=GetOwnerSkeletalMeshComponent())
   > Set avoidClippingbkNecklace(avoidClippingbkNecklace:double=GetAlphaFromReferenceTransformDelta(SkeletalMeshComponent=@n3, BoneName=head, BaseName=neck_01, InputRange=(X=-5,Y=-20), TwistAxis="0,0,1")) -> (Output_Get:double)
   > Set Head_Turn_R(Head_Turn_R:double=GetAlphaFromReferenceTransformDelta(SkeletalMeshComponent=@n3, BoneName=spine_05, BaseName=head, DeltaComponentType=EulerX, InputRange=(X=0,Y=-20), TwistAxis="1,0,0")) -> (Output_Get:double)
   > Set Head_Turn_L(Head_Turn_L:double=GetAlphaFromReferenceTransformDelta(SkeletalMeshComponent=@n3, BoneName=spine_05, BaseName=head, DeltaComponentType=EulerX, InputRange=(X=0,Y=20), TwistAxis="1,0,0")) -> (Output_Get:double)
   n4 = BreakTransform(InTransform=GetDeltaTransformFromRefPose(self=GetOwnerSkeletalMeshComponent(), BoneName=spine_05, BaseName=spine_01))
   n5 = MakeRotator(Pitch=float * float(A=BreakRotator(InRot=@n4.Rotation), B=0.38) {is=title K2Node_PromotableOperator})
   Set spine rotation(execute=@n2.then_6, spine rotation:Rotator=@n5) -> (Output_Get:Rotator)
   > Set spine rotation_2(spine rotation_2:Rotator="Scale Rotator (Float)"(A:Rotator=@n5, B:float=-3) -> (ReturnValue:Rotator=~) {is=title K2Node_PromotableOperator}) -> (Output_Get:Rotator)
   n6 = Gravity Override State Function DEPRECATED(
    execute=@n2.then_1
    CharacterPartAnimInstance=Self-Reference()
    GameTimeInSeconds=GetGameTimeInSeconds()
    GravityOverrideParams=Get Gravity Override Params()
   )
   > Set GravityOverride(GravityOverride=@n6)
   n7 = GetOwnerSkeletalMeshComponent()
   n8 = BreakRotator(InRot=BreakTransform(InTransform=GetDeltaTransformFromRefPose(self=@n7, BoneName=head, BaseName=neck_02)).Rotation)
   n9 = Get Y percentage() -> (Y percentage:float)
   n10 = Get X percentage() -> (X percentage:float)
   n11 = Get Z percentage() -> (Z percentage:float)
   > Set spine rotation(spine rotation=MakeRotator(
     Roll=float * float!(A=@n8, B=-1, C:double=@n10) -> (ReturnValue) {is=title K2Node_CommutativeAssociativeBinaryOperator}
     Pitch=float * float(A=@n8.Pitch, B=-1, C=@n9)
     Yaw=float * float(A=@n8.Yaw, B=-1, C=@n11)
    ))
   > Set SkydiveDiveAlpha(SkydiveDiveAlpha:double="To Float (Boolean)"(InBool=Get bIsSkydiveDiveMode())) -> (Output_Get:double)
   > Set ParachuteAlpha(ParachuteAlpha:double="To Float (Boolean)"(InBool=Get bIsParachuteOpened())) -> (Output_Get:double)
   > Set BaseballSliding(BaseballSliding:double="To Float (Boolean)"(InBool=Get bIsBaseballSliding())) -> (Output_Get:double)
   > Set aimingalpha(aimingalpha:double="To Float (Boolean)"(InBool=Get bIsTargeting())) -> (Output_Get:double)
   n12 = Get bIsPlayingMeleeAnim()
   > Set usingPickaxeAlpha(usingPickaxeAlpha:double=FInterpTo(
     Current=Get usingPickaxeAlpha() -> (usingPickaxeAlpha:double)
     Target="To Float (Boolean)"(InBool=@n12)
     DeltaTime=@n1.DeltaTimeX
     InterpSpeed=0.1
    )) -> (Output_Get:double)
   n13 = NAND Boolean(A=Get bIsCrouching(), B=Get bIsMoving2D()) {is=title K2Node_CommutativeAssociativeBinaryOperator}
   > Set CrouchIdleAlpha(CrouchIdleAlpha:double=FInterpTo(Current=Get CrouchIdleAlpha(), Target="To Float (Boolean)"(InBool=@n13), DeltaTime=@n1.DeltaTimeX, InterpSpeed=5)) -> (Output_Get:double)
   > Set KneeSlidingAlpha(KneeSlidingAlpha:double="To Float (Boolean)"(InBool=Get bIsKneeSliding())) -> (Output_Get:double)
   n14 = Get bIsSkydiveFloating()
   n15 = Get bIsSkydiveDiveMode()
   > Set SkydiveFloatingAlpha(SkydiveFloatingAlpha:double="To Float (Boolean)"(InBool=OR Boolean(A=@n14, B=@n15))) -> (Output_Get:double)
   > Set InAir(InAir:double="To Float (Boolean)"(InBool=Get bIsSkyDiving())) -> (Output_Get:double)
   > Set SwimmingAlpha(SwimmingAlpha:double="To Float (Boolean)"(InBool=Get bIsSwimSprinting() -> (bIsSwimSprinting:bool))) -> (Output_Get:double)
   n16 = GetDeltaTransformFromRefPose(self=GetOwnerSkeletalMeshComponent(), BoneName=neck_02, BaseName=neck_01)
   n17 = "float < float"(A=BreakRotator(InRot=BreakTransform(InTransform=@n16).Rotation).Pitch, B=8)
   Set HeadRotL(
    execute=@n2.then_3
    HeadRotL:double=FInterpTo(Current=Get HeadRotL(), Target="To Float (Boolean)"(InBool=@n17), DeltaTime=@n1.DeltaTimeX, InterpSpeed=5)
   ) -> (Output_Get:double)
   n18 = Get bRigidBodyEnabled()
   > Set NOT RigidBody(NOT RigidBody:double="To Float (Boolean)"(InBool=NOT Boolean(A=@n18))) -> (Output_Get:double)
   > n19 = Sequence!(execute) -> (then_0, then_1, then_2, then_3)
   > n20 = Branch(Condition=NOR Boolean!(A:bool=Get bIsTargeting(), B:bool=Get bIsMoving2D()) -> (ReturnValue:bool))
   > Set idleAlpha(idleAlpha:double=5) -> (Output_Get:double)
   Set idleAlpha(execute=@n20.else, idleAlpha=1)
   > n21 = Branch(Condition=OR Boolean(A=Get bIsOnGrindRail(), B=Get OnZipline() -> (OnZipline:bool)))
   > Set RailAlpha(RailAlpha:double=0.6) -> (Output_Get:double)
   Set RailAlpha(execute=@n21.else, RailAlpha=0)
   n22 = Branch(execute=@n19.then_1, Condition=Get bIsDBNO())
   > Set DBNOAlpha(DBNOAlpha:double=0.4) -> (Output_Get:double)
   Set DBNOAlpha(execute=@n22.else, DBNOAlpha=1)
   n23 = Branch(execute=@n19.then_2, Condition=Get OnZipline())
   > Set ZiplineAlpha(ZiplineAlpha:double=0) -> (Output_Get:double)
   Set ZiplineAlpha(execute=@n23.else, ZiplineAlpha=1)
   n24 = Sequence!(execute=@n2.then_0) -> (then_0, then_1)
   > n25 = Cast To FortPlayerAnimInstance(Object=GetAnimInstance(self=GetOwnerSkeletalMeshComponent()))
   > Set AlphaPistol(AlphaPistol:float=1) -> (Output_Get:float)
   > Set shotgunEquipped(shotgunEquipped:double=1) -> (Output_Get:double)
   > Set rifle_out(rifle_out:bool=false) -> (Output_Get:bool)
   > Set PistolOut(PistolOut:bool=false) -> (Output_Get:bool)
   > n26 = Switch on EFortWeaponCoreAnimation!(execute, Selection:EFortWeaponCoreAnimation Enum=Get WeaponCoreAnim(self=@n25)) -> (
    Melee:Exec
    Pistol:Exec
    Shotgun:Exec
    PaperBlueprint:Exec
    Rifle:Exec
    MeleeOneHand:Exec
    MachinePistol:Exec
    RocketLauncher:Exec
    GrenadeLauncher:Exec
    GoingCommando:Exec
    AssaultRifle:Exec
    TacticalShotgun:Exec
    SniperRifle:Exec
    TrapPlacement:Exec
    ShoulderLauncher:Exec
    AbilityDecoTool:Exec
    Crossbow:Exec
    C4:Exec
    RemoteControl:Exec
    DualWield:Exec
    AR_BullPup:Exec
    AR_ForwardGrip:Exec
    MedPackPaddles:Exec
    SMG_P90:Exec
    AR_DrumGun:Exec
    Consumable_Small:Exec
    Consumable_Large:Exec
    Balloon:Exec
    MountedTurret:Exec
    CreativeTool:Exec
    ExplosiveBow:Exec
    AshtonIndigo:Exec
    AshtonChicago:Exec
    MeleeDualWield:Exec
    Unarmed:Exec
   )
   Set AlphaPistol(execute=@n26.Pistol, AlphaPistol=1.5)
   n27 = Sequence(execute=@n26.Shotgun)
   > Set shotgunEquipped(shotgunEquipped=0.4)
   Branch(execute=[@n27.then_1, @n26.AssaultRifle])
   > Set rifle_out(rifle_out=true)
   Set large_gun_out(execute=@n24.then_1, large_gun_out:double="To Float (Boolean)"(InBool=Get rifle_out() -> (rifle_out:bool))) -> (Output_Get:double)
   n28 = PlayerMovement(execute=@n2.then_2, AnimInstance=Self-Reference(), XDotProductCoefficient=100, YDotProductCoefficient=100, ZDotProductCoefficient=100)
   n29 = BreakVector(InVec=MakeVector(X=@n28, Y=@n28.YMovement))
   > n30 = Set LeftBool(LeftBool:bool="float < float"(A=@n29, B=-50)) -> (Output_Get:bool)
   > n31 = Set RightBool(RightBool:bool="float > float"(A=@n29, B=50)) -> (Output_Get:bool)
   > Set forwardBool(forwardBool:bool="float > float"(A=@n29.Y, B=40)) -> (Output_Get:bool)
   > Set BackwardBool(BackwardBool:bool="float < float"(A=@n29.Y, B=-40)) -> (Output_Get:bool)
   > n32 = Branch(Condition=AND Boolean(A=OR Boolean(A=@n30, B=@n31), B=Get bIsCrouching()))
   > Set SideToSide(SideToSide:double=1) -> (Output_Get:double)
   Set SideToSide(execute=@n32.else, SideToSide=0.2)
   n33 = NOT Boolean(A=Get bIsPlayingEmote())
   n34 = NOT Boolean(A=Get bIsSkyDiving())
   n35 = Get bRigidBodyEnabled()
   n36 = Get bIsInFrontEnd()
   Set FrontendAlpha(
    execute=@n2.then_5
    FrontendAlpha:double="To Float (Boolean)"(InBool=AND Boolean!(A=@n36, B=@n34, C:bool=@n33, D:bool=@n35) -> (ReturnValue))
   ) -> (Output_Get:double)
   n37 = Event BlueprintProcessAnimInputEvent()
   > n38 = Branch(Condition=IsNamed(self=@n37.AnimInputEvent, InEventName=aie.Player.Zipline.Entered))
   > Set OnZipline(OnZipline:bool=1) -> (Output_Get:bool)
   Branch(execute=@n38.else, Condition=IsNamed(self=@n37.AnimInputEvent, InEventName=aie.Player.Zipline.Exited))
   > Set OnZipline(OnZipline=false)
   Event OnCharacterCustomizationLoaded()
   Set NewShotgun(NewShotgun:bool=false) -> (Output_Get:bool)
  ]}
]
variables [
 aimingalpha {type=number}
 AlphaPistol {type=number}
 avoidClippingbkNecklace {type=number}
 BackwardBool {type=boolean}
 BaseballSliding {type=number}
 CrouchIdleAlpha {type=number}
 DBNOAlpha {type=number}
 forwardBool {type=boolean}
 FrontendAlpha {type=number}
 Gravity Override Params {type=GravityOverrideParamsStruct}
 GravityOverride {type=Vector}
 Head_Turn_L {type=number}
 Head_Turn_R {type=number}
 HeadRotL {type=number}
 idleAlpha {type=number}
 InAir {type=number}
 KneeSlidingAlpha {type=number}
 large_gun_out {type=number}
 LeftBool {type=boolean}
 NewShotgun {type=boolean}
 NOT RigidBody {type=number}
 OnZipline {type=boolean}
 ParachuteAlpha {type=number}
 PistolOut {type=boolean}
 RailAlpha {type=number}
 rifle_out {type=boolean}
 RightBool {type=boolean}
 shotgunEquipped {type=number}
 SideToSide {type=number}
 Skel Mesh {type=SkeletalMeshComponent}
 SkydiveDiveAlpha {type=number}
 SkydiveFloatingAlpha {type=number}
 spine rotation {type=Rotator}
 spine rotation_2 {type=Rotator}
 SpineRot {type=number}
 SwimmingAlpha {type=number}
 translation_necklace {type=Vector}
 usingPickaxeAlpha {type=number}
 X percentage {type=number, default=0.5}
 Y percentage {type=number}
 Z percentage {type=number}
 ZiplineAlpha {type=number}
]
