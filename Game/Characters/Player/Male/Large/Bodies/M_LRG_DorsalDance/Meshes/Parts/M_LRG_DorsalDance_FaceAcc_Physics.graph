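# Physics asset for the face accessory mesh of the large male DorsalDance player body, covering a cap and glasses.
# Each accessory pairs a tiny capsule on its base joint with a larger, offset capsule on its dyn joint.
# One constraint per pair links dyn to base, with zero twist and first-swing limits and a small second-swing limit, suggesting a slight single-axis wobble.
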PhysicsAsset /Game/Characters/Player/Male/Large/Bodies/M_LRG_DorsalDance/Meshes/Parts/M_LRG_DorsalDance_FaceAcc_Physics
bodies {
 base_cap_jnt [{shapeType=Capsule, length=0.505, radius=0.505, shapeName=base_cap_jnt_capsule, rotation=<rot 0 -90 0>}]
 base_glasses_jnt [{shapeType=Capsule, length=0.505, radius=0.505, shapeName=base_glasses_jnt_capsule, rotation=<rot 0 -90 0>}]
 dyn_cap_jnt [{shapeType=Capsule, length=3, radius=3, shapeName=dyn_cap_jnt_capsule, center=<2 0 0>, rotation=<rot 90 0 0>}]
 dyn_glasses_jnt [{shapeType=Capsule, length=2, radius=1, shapeName=dyn_glasses_jnt_capsule, center=<2 0 0>, rotation=<rot 90 0 0>}]
}
constraints [
 {bone1Name=dyn_glasses_jnt, bone2Name=base_glasses_jnt, swing1LimitDegrees=0, swing2LimitDegrees=10, twistLimitDegrees=0}
 {bone1Name=dyn_cap_jnt, bone2Name=base_cap_jnt, swing1LimitDegrees=0, swing2LimitDegrees=6, twistLimitDegrees=0}
]
